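# Control Rig for the F_MED_ToadCycle female body cosmetic, adding corrective deformation to its skirt bones as the legs move.
# Setup creates pelvis-parented thigh reference nulls; thigh rotation against these is remapped into side weights.
# After a shared DeformRig pass, those weights and deformRig knee and thigh attributes blend skirt offsets, alongside a constant scarf nudge.
# When isSliding is set, back-skirt-to-right-foot distance blends in further skirt offsets.
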ControlRigBlueprint /BRCosmetics/Characters/Player/Female/Medium/Bodies/F_MED_ToadCycle/Meshes/F_MED_ToadCycle_CtrlRig
graphs [
 New Function_ContainedGraph {nodes [
   RigVMFunctionEntryNode()
   > RigVMFunction_Sequence = RigVMAggregateNode()
   &RigUnit_ModifyTransforms(
    ItemToModify="((Item=(Type=Bone,Name=\"dyn_skirt_b_l_base\"),Transform=(Rotation=(X=0,Y=-0.0872,Z=0,W=0.9962),Translation=(X=0,Y=0,Z=0),Scale3D=(X=1,Y=1,Z=1))),(Item=(Type=Bone,Name=\"dyn_skirt_s_l_base\"),Transform=(Rotation=(X=0,Y=-0.0872,Z=0,W=0.9962),Translation=(X=0,Y= ...(truncated)"
    WeightMinimum=0
    Weight=RigVMDispatchNode(BoneName=root, Default=0, Name=deformRig_l_knee_alpha) {template="DISPATCH_RigDispatch_GetAnimAttribute(in Name,in BoneName,in Default,out Value,out Found)"}.Value
    ExecutePin=@RigVMFunction_Sequence.A
   )
   > &RigUnit_ModifyTransforms(
    ItemToModify="((Item=(Type=Bone,Name=\"dyn_skirt_b_r_base\"),Transform=(Rotation=(X=0,Y=-0.1736,Z=0,W=0.9848),Translation=(X=0,Y=0,Z=0),Scale3D=(X=1,Y=1,Z=1))),(Item=(Type=Bone,Name=\"dyn_skirt_s_r_base\"),Transform=(Rotation=(X=0,Y=-0.1736,Z=0,W=0.9848),Translation=(X=0,Y= ...(truncated)"
    WeightMinimum=0
    Weight=RigVMDispatchNode(BoneName=root, Default=0, Name=deformRig_r_knee_alpha) {template="DISPATCH_RigDispatch_GetAnimAttribute(in Name,in BoneName,in Default,out Value,out Found)"}.Value
   )
   &RigUnit_ModifyTransforms(
    ItemToModify="((Item=(Type=Bone,Name=\"dyn_skirt_f_l_base\"),Transform=(Rotation=(X=0,Y=-0.766,Z=0,W=0.6428),Translation=(X=-7,Y=0,Z=0),Scale3D=(X=1,Y=1,Z=1))),(Item=(Type=Bone,Name=\"dyn_skirt_s_l_base\"),Transform=(Rotation=(X=0.2932,Y=-0.7077,Z=0.246,W=0.5939),Translatio ...(truncated)"
    WeightMinimum=0
    ExecutePin=@RigVMFunction_Sequence.B
    Weight=RigVMVariableNode() {variable=ThighL_Side}.Value
   )
   > &RigUnit_ModifyTransforms(
    ItemToModify="((Item=(Type=Bone,Name=\"dyn_skirt_f_r_base\"),Transform=(Rotation=(X=0,Y=-0.766,Z=0,W=0.6428),Translation=(X=-3,Y=0,Z=0),Scale3D=(X=1,Y=1,Z=1))),(Item=(Type=Bone,Name=\"dyn_skirt_s_r_base\"),Transform=(Rotation=(X=-0.2932,Y=-0.7077,Z=-0.246,W=0.5939),Translat ...(truncated)"
    WeightMinimum=0
    Weight=RigVMVariableNode() {variable=ThighR_Side}.Value
   )
   &RigUnit_ModifyTransforms(
    ItemToModify="((Item=(Type=Bone,Name=\"dyn_skirt_f2_l_base\"),Transform=(Rotation=(X=-0.1464,Y=-0.3536,Z=-0.3536,W=0.8536),Translation=(X=-3,Y=0,Z=0),Scale3D=(X=1,Y=1,Z=1))),(Item=(Type=Bone,Name=\"dyn_skirt_f_l_base\"),Transform=(Rotation=(X=0.1464,Y=-0.3536,Z=0.3536,W=0.8 ...(truncated)"
    WeightMinimum=0
    Weight=RigVMDispatchNode(BoneName=root, Default=0, Name=deformRig_l_thigh_up_alpha) {template="DISPATCH_RigDispatch_GetAnimAttribute(in Name,in BoneName,in Default,out Value,out Found)"}.Value
    ExecutePin=@RigVMFunction_Sequence.C
   )
   > &RigUnit_ModifyTransforms(
    ItemToModify="((Item=(Type=Bone,Name=\"dyn_skirt_f2_r_base\"),Transform=(Rotation=(X=0,Y=-0.3827,Z=0,W=0.9239),Translation=(X=-3,Y=0,Z=0),Scale3D=(X=1,Y=1,Z=1))),(Item=(Type=Bone,Name=\"dyn_skirt_f_r_base\"),Transform=(Rotation=(X=-0.1464,Y=-0.3536,Z=-0.3536,W=0.8536),Trans ...(truncated)"
    WeightMinimum=0
    Weight=RigVMDispatchNode(BoneName=root, Default=0, Name=deformRig_r_thigh_up_alpha) {template="DISPATCH_RigDispatch_GetAnimAttribute(in Name,in BoneName,in Default,out Value,out Found)"}.Value
   )
   > &RigUnit_ModifyTransforms(
    ItemToModify=((Item=(Type=Bone,Name="base_scarf_fr"),Transform=(Rotation=(X=0,Y=0,Z=0,W=1),Translation=(X=0,Y=0,Z=0.5),Scale3D=(X=1,Y=1,Z=1))))
    Weight=1
    WeightMinimum=0
   )
   RigVMFunctionReturnNode(ExecuteContext=@RigVMFunction_Sequence.D)
  ]}
 New Function_ContainedGraph {nodes [
   n1 = RigVMFunctionEntryNode()
   RigVMFunction_MathQuaternionToEuler = &RigVMFunction_MathQuaternionToEuler(
    RotationOrder=XYZ
    Value=&RigUnit_GetRelativeTransformForItem(
     Child=(Type=Bone,Name="hand_l")
     Parent=(Type=Null,Name="lowerarm_l")
     bChildInitial=False
     bParentInitial=False
     Child.Name=@n1.Child
     Parent.Name=@n1.Parent
    ).RelativeTransform.Rotation
   )
   Remap = &RigVMFunction_MathDoubleRemap(
    TargetMaximum=1
    TargetMinimum=0
    Value=@RigVMFunction_MathQuaternionToEuler.Result.Y
    SourceMinimum=@n1.Source Minimum
    SourceMaximum=@n1.Source Maximum
   )
   > RigVMFunctionReturnNode(Delta Rotation=@Remap.Result)
  ]}
 New Function_ContainedGraph {nodes [
   n1 = RigVMFunctionEntryNode()
   RigVMFunction_Sequence = &RigVMFunction_Sequence(ExecuteContext=@n1.ExecuteContext)
   DeltaRotY = RigVMFunctionReferenceNode(Child=thigh_l, Parent=Ref_Thigh_L, Source Maximum=-70, Source Minimum=-5, ExecuteContext=@RigVMFunction_Sequence.A)
   > RigVMVariableNode(Value=@DeltaRotY.Delta Rotation) {variable=ThighL_Side}
   > DeltaRotY_1 = RigVMFunctionReferenceNode(Child=thigh_r, Parent=Ref_Thigh_R, Source Maximum=-70, Source Minimum=-5)
   > RigVMVariableNode(Value=@DeltaRotY_1.Delta Rotation) {variable=ThighR_Side}
   RigVMFunctionReturnNode(ExecuteContext=@RigVMFunction_Sequence.B)
  ]}
 New Function_ContainedGraph {nodes [
   n1 = RigVMFunctionEntryNode()
   RigVMFunction_ControlFlowBranch = &RigVMFunction_ControlFlowBranch(ExecuteContext=@n1.ExecuteContext, Condition=RigVMVariableNode() {variable=isSliding}.Value)
   RigVMFunctionReturnNode(ExecuteContext=@RigVMFunction_ControlFlowBranch.Completed)
   RigVMFunction_MathVectorDistance = &RigVMFunction_MathVectorDistance(
    A=&RigUnit_GetTransform(Item=(Type=Bone,Name="dyn_skirt_b_l_01"), bInitial=False).Transform.Translation
    B=&RigUnit_GetTransform(Item=(Type=Bone,Name="foot_r"), bInitial=False).Transform.Translation
   )
   Remap = &RigVMFunction_MathDoubleRemap(SourceMaximum=48, SourceMinimum=50, TargetMaximum=1, TargetMinimum=0, Value=@RigVMFunction_MathVectorDistance.Result)
   &RigUnit_ModifyTransforms(
    ItemToModify="((Item=(Type=Bone,Name=\"dyn_skirt_b_l_base\"),Transform=(Rotation=(X=0,Y=-0.7071,Z=0,W=0.7071),Translation=(X=-3,Y=0,Z=0),Scale3D=(X=1,Y=1,Z=1))),(Item=(Type=Bone,Name=\"dyn_skirt_b_r_base\"),Transform=(Rotation=(X=0,Y=-0.7071,Z=0,W=0.7071),Translation=(X=-3, ...(truncated)"
    WeightMinimum=0
    Weight=@Remap.Result
    ExecutePin=@RigVMFunction_ControlFlowBranch.True
   )
   > &RigUnit_ModifyTransforms(
    ItemToModify="((Item=(Type=Bone,Name=\"None\"),Transform=(Rotation=(X=0,Y=0,Z=0,W=1),Translation=(X=0,Y=0,Z=0),Scale3D=(X=1,Y=1,Z=1))),(Item=(Type=Bone,Name=\"dyn_skirt_b_r_base\"),Transform=(Rotation=(X=0,Y=0,Z=-0.3827,W=0.9239),Translation=(X=0,Y=0,Z=0),Scale3D=(X=1,Y=1,Z ...(truncated)"
    WeightMinimum=0
    Weight=@Remap.Result
   )
  ]}
 RigVMFunctionLibrary {nodes [
   DeltaRotY = RigVMCollapseNode(ExecuteContext=())
   DressSetup = RigVMCollapseNode(ExecuteContext=())
   Variables = RigVMCollapseNode(ExecuteContext=())
   isSliding = RigVMCollapseNode(ExecuteContext=())
  ]}
 RigVMFunction_Sequence_1_ContainedGraph {nodes [
   n1 = RigVMFunctionEntryNode()
   RigVMFunction_Sequence = &RigVMFunction_Sequence(ExecuteContext=@n1.ExecuteContext)
   n2 = &RigVMFunction_Sequence(ExecuteContext=@RigVMFunction_Sequence.B)
   n3 = &RigVMFunction_Sequence(ExecuteContext=@n2.B)
   RigVMFunctionReturnNode(A=@RigVMFunction_Sequence.A, B=@n2.A, C=@n3.A, D=@n3.B)
  ]}
 RigVMModel {nodes [
   n1 = &RigUnit_PrepareForExecution()
   &RigUnit_HierarchyImportFromSkeleton(NameSpace=None, bIncludeCurves=False, bIncludeMeshSockets=False, bIncludeVirtualBones=True, ExecutePin=@n1.ExecutePin)
   > &RigUnit_HierarchyAddNull(
    Name=Ref_Thigh_L
    Parent=(Type=Bone,Name="pelvis")
    Space=GlobalSpace
    Transform=&RigUnit_GetTransform(Item=(Type=Bone,Name="thigh_twist_01_l"), bInitial=False).Transform
   )
   > &RigUnit_HierarchyAddNull(
    Name=Ref_Thigh_R
    Parent=(Type=Bone,Name="pelvis")
    Space=GlobalSpace
    Transform=&RigUnit_GetTransform(Item=(Type=Bone,Name="thigh_twist_01_r"), bInitial=False).Transform
   )
   RigUnit_BeginExecution = &RigUnit_BeginExecution()
   n2 = &RigVMFunction_Sequence(ExecuteContext=@RigUnit_BeginExecution.ExecutePin)
   FN_DeformRig_FNC_1 = RigVMFunctionReferenceNode(
    DeformRig_ Curve="(deformRig_l_upperarm_up_alpha_72_#1=False,deformRig_r_upperarm_up_alpha_73_#2=False,deformRig_l_upperarm_fr_alpha_74_#3=False,deformRig_r_upperarm_fr_alpha_75_CB51B8 ...(truncated)"
    DeformRig_Bool="(upperarm_1_#4=True,elbow_3_#5=True,wrist_7_#6=True,groin_8_#7=True,glute_17_#8=True,knee_9_40C408324CA261 ...(truncated)"
    DeformRig_Debug=(upperarm_debug_18_#4=False,wrist_debug_20_#6=False,lat_debug_26_#9=False,groinGlute_debug_27_#7=False)
    DeformRig_Value="(upperarm_up_18_#10=(Rotation=(X=0,Y=0,Z=0,W=1),Translation=(X=0,Y=0,Z=8),Scale3D=(X=0.7,Y=1,Z=1)),upperarm_fwd_fr_19_#11=(Rotation=(X=0,Y=0,Z=0,W=1),Translation=(X=0,Y=10,Z=0),Scale3D=(X=0.5,Y=1,Z= ...(truncated)"
    ExecuteContext=@n2.A
   )
   > Variables = RigVMFunctionReferenceNode()
   > DressSetup = RigVMFunctionReferenceNode()
   isSliding = RigVMFunctionReferenceNode(ExecuteContext=@n2.B)
  ]}
]
variables [
 ThighL_Side {type=double, default=0}
 ThighR_Side {type=double, default=0}
 isSliding {type=bool, default=False}
]
