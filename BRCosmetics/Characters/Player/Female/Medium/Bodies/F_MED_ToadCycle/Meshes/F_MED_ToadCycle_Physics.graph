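# Physics asset for the F_MED_ToadCycle female player body, built for physics on its skirt bones.
# Nine three-segment capsule chains hang from small base bodies tied to the pelvis with free rotation, with tight swing and twist limits down each chain.
# Cross-constraints with mostly free rotation link some neighboring chains.
# Pelvis, thigh, calf and foot capsules supply body shapes, and a huge root sphere below forms a floor-like surface.
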PhysicsAsset /BRCosmetics/Characters/Player/Female/Medium/Bodies/F_MED_ToadCycle/Meshes/F_MED_ToadCycle_Physics
bodies {
 calf_l [{shapeType=Capsule, length=30.541, radius=7, shapeName=thigh_l_capsule, center=<-20 -1.016 -1>, rotation=<rot 0.0806 -90.8115 -89.0784>}]
 calf_r [{shapeType=Capsule, length=30.541, radius=7, shapeName=thigh_l_capsule, center=<20 0 1>, rotation=<rot 0.0806 -90.8115 -89.0784>}]
 dyn_skirt_b_l_01 [{shapeType=Capsule, length=10, radius=9, center=<5 0 7>, rotation=<rot 90 90 0>}]
 dyn_skirt_b_l_02 [{shapeType=Capsule, length=10, radius=9, center=<5 0 5>, rotation=<rot 90 90 0>}]
 dyn_skirt_b_l_03 [{shapeType=Capsule, length=10, radius=9, center=<5 0 5>, rotation=<rot 90 0 0>}]
 dyn_skirt_b_l_base [{shapeType=Capsule, length=0.505, radius=0.505, shapeName=dyn_skirt_b_l_base_capsule, rotation=<rot 0 -90 0>}]
 dyn_skirt_b_r_01 [{shapeType=Capsule, length=10, radius=9, center=<5 0 7>, rotation=<rot 90 90 0>}]
 dyn_skirt_b_r_02 [{shapeType=Capsule, length=10, radius=9, center=<5 0 5>, rotation=<rot 90 90 0>}]
 dyn_skirt_b_r_03 [{shapeType=Capsule, length=10, radius=9, center=<5 -5 5>, rotation=<rot 90 0 90>}]
 dyn_skirt_b_r_base [{shapeType=Capsule, length=0.505, radius=0.505, shapeName=dyn_skirt_b_r_base_capsule, rotation=<rot 0 -90 0>}]
 dyn_skirt_f2_l_01 [{shapeType=Capsule, length=10, radius=8, center=<5 0 7>, rotation=<rot 90 90 0>}]
 dyn_skirt_f2_l_02 [{shapeType=Capsule, length=10, radius=9, center=<5 0 7>, rotation=<rot 90 90 0>}]
 dyn_skirt_f2_l_03 [{shapeType=Capsule, length=10, radius=10, center=<5 0 7>, rotation=<rot 90 90 0>}]
 dyn_skirt_f2_l_base [{shapeType=Capsule, length=0.505, radius=0.505, shapeName=dyn_skirt_f2_l_base_capsule, rotation=<rot 0 -90 0>}]
 dyn_skirt_f2_r_01 [{shapeType=Capsule, length=10, radius=8, center=<5 0 7>, rotation=<rot 90 90 0>}]
 dyn_skirt_f2_r_02 [{shapeType=Capsule, length=10, radius=9, center=<5 0 5>, rotation=<rot 90 90 0>}]
 dyn_skirt_f2_r_03 [{shapeType=Capsule, length=10, radius=9, center=<5 0 5>, rotation=<rot 90 90 0>}]
 dyn_skirt_f2_r_base [{shapeType=Capsule, length=0.505, radius=0.505, shapeName=dyn_skirt_f2_r_base_capsule, rotation=<rot 0 -90 0>}]
 dyn_skirt_f_c_01 [{shapeType=Capsule, length=10, radius=9, center=<5 0 7>, rotation=<rot 90 90 0>}]
 dyn_skirt_f_c_02 [{shapeType=Capsule, length=10, radius=9, center=<5 0 7>, rotation=<rot 90 90 0>}]
 dyn_skirt_f_c_03 [{shapeType=Capsule, length=10, radius=9, center=<5 0 7>, rotation=<rot 90 90 0>}]
 dyn_skirt_f_c_base [{shapeType=Capsule, length=0.505, radius=0.505, shapeName=dyn_skirt_f_c_base_capsule, rotation=<rot 0 -90 0>}]
 dyn_skirt_f_l_01 [{shapeType=Capsule, length=10, radius=9, center=<5 0 7>, rotation=<rot 90 90 0>}]
 dyn_skirt_f_l_02 [{shapeType=Capsule, length=10, radius=9, center=<5 0 7>, rotation=<rot 90 90 0>}]
 dyn_skirt_f_l_03 [{shapeType=Capsule, length=10, radius=9, center=<5 0 7>, rotation=<rot 90 90 0>}]
 dyn_skirt_f_l_base [{shapeType=Capsule, length=0.505, radius=0.505, shapeName=dyn_skirt_f_l_base_capsule, rotation=<rot 0 -90 0>}]
 dyn_skirt_f_r_01 [{shapeType=Capsule, length=10, radius=9, center=<5 0 7>, rotation=<rot 90 90 0>}]
 dyn_skirt_f_r_02 [{shapeType=Capsule, length=10, radius=9, center=<5 0 7>, rotation=<rot 90 90 0>}]
 dyn_skirt_f_r_03 [{shapeType=Capsule, length=10, radius=9, center=<5 0 7>, rotation=<rot 90 90 0>}]
 dyn_skirt_f_r_base [{shapeType=Capsule, length=0.505, radius=0.505, shapeName=dyn_skirt_f_r_base_capsule, rotation=<rot 0 -90 0>}]
 dyn_skirt_s_l_01 [{shapeType=Capsule, length=10, radius=9, center=<5 0 7>, rotation=<rot 90 90 0>}]
 dyn_skirt_s_l_02 [{shapeType=Capsule, length=10, radius=9, center=<5 0 5>, rotation=<rot 90 90 0>}]
 dyn_skirt_s_l_03 [{shapeType=Capsule, length=10, radius=9, center=<5 0 5>, rotation=<rot 90 90 0>}]
 dyn_skirt_s_l_base [{shapeType=Capsule, length=0.505, radius=0.505, shapeName=dyn_skirt_s_l_base_capsule, rotation=<rot 0 -90 0>}]
 dyn_skirt_s_r_01 [{shapeType=Capsule, length=10, radius=9, center=<5 0 7>, rotation=<rot 90 90 0>}]
 dyn_skirt_s_r_02 [{shapeType=Capsule, length=10, radius=9, center=<5 0 5>, rotation=<rot 90 90 0>}]
 dyn_skirt_s_r_03 [{shapeType=Capsule, length=10, radius=9, center=<5 0 5>, rotation=<rot 90 90 0>}]
 dyn_skirt_s_r_base [{shapeType=Capsule, length=0.505, radius=0.505, shapeName=dyn_skirt_s_r_base_capsule, rotation=<rot 0 -90 0>}]
 foot_l [{shapeType=Capsule, length=11.0044, radius=11.0044, shapeName=foot_l_capsule, center=<-5 8 0>, rotation=<rot 0 0 -90>}]
 foot_r [{shapeType=Capsule, length=11.0044, radius=11.0044, shapeName=foot_r_capsule, center=<5 -8 0>, rotation=<rot 0 0 -90>}]
 pelvis [{shapeType=Capsule, length=10.7667, radius=9, shapeName=pelvis_capsule, center=<-0.4976 -0.6252 0>, rotation=<rot 0 -90.0003 0>}]
 root [{shapeType=Sphere, radius=500, shapeName=root_sphere, center=<0 17.4691 -494.0154>}]
 thigh_l [{
   shapeType=Capsule
   length=30.541
   radius=7.2
   shapeName=thigh_l_capsule
   center=<-23.7262 -0.9614 0.3035>
   rotation=<rot 0.0806 -90.8116 -94.0784>
  }]
 thigh_r [{
   shapeType=Capsule
   length=30.541
   radius=7.2
   shapeName=thigh_l_capsule
   center=<23.4593 -0.0503 0.0922>
   rotation=<rot 0.0806 -90.8117 -94.0784>
  }]
}
constraints [
 {
  bone1Name=dyn_skirt_b_l_base
  bone2Name=pelvis
  swing1LimitDegrees=45
  swing1Motion=Free
  swing2LimitDegrees=45
  swing2Motion=Free
  twistLimitDegrees=45
  twistMotion=Free
 }
 {bone1Name=dyn_skirt_b_l_01, bone2Name=dyn_skirt_b_l_base, swing1LimitDegrees=10, swing2LimitDegrees=20, twistLimitDegrees=5}
 {bone1Name=dyn_skirt_b_l_02, bone2Name=dyn_skirt_b_l_01, swing1LimitDegrees=10, swing2LimitDegrees=20, twistLimitDegrees=5}
 {
  bone1Name=dyn_skirt_b_l_02
  bone2Name=dyn_skirt_b_r_02
  swing1LimitDegrees=45
  swing1Motion=Free
  swing2LimitDegrees=45
  swing2Motion=Free
  twistLimitDegrees=45
  twistMotion=Free
 }
 {bone1Name=dyn_skirt_b_l_03, bone2Name=dyn_skirt_b_l_02, swing1LimitDegrees=10, swing2LimitDegrees=20, twistLimitDegrees=5}
 {
  bone1Name=dyn_skirt_b_r_base
  bone2Name=pelvis
  swing1LimitDegrees=45
  swing1Motion=Free
  swing2LimitDegrees=45
  swing2Motion=Free
  twistLimitDegrees=45
  twistMotion=Free
 }
 {bone1Name=dyn_skirt_b_r_01, bone2Name=dyn_skirt_b_r_base, swing1LimitDegrees=10, swing2LimitDegrees=20, twistLimitDegrees=5}
 {bone1Name=dyn_skirt_b_r_02, bone2Name=dyn_skirt_b_r_01, swing1LimitDegrees=10, swing2LimitDegrees=20, twistLimitDegrees=5}
 {bone1Name=dyn_skirt_b_r_03, bone2Name=dyn_skirt_b_r_02, swing1LimitDegrees=10, swing2LimitDegrees=20, twistLimitDegrees=5}
 {
  bone1Name=dyn_skirt_f2_l_base
  bone2Name=pelvis
  swing1LimitDegrees=45
  swing1Motion=Free
  swing2LimitDegrees=45
  swing2Motion=Free
  twistLimitDegrees=45
  twistMotion=Free
 }
 {bone1Name=dyn_skirt_f2_l_01, bone2Name=dyn_skirt_f2_l_base, swing1LimitDegrees=10, swing2LimitDegrees=20, twistLimitDegrees=5}
 {
  bone1Name=dyn_skirt_f2_l_01
  bone2Name=dyn_skirt_f_l_01
  swing1LimitDegrees=45
  swing1Motion=Free
  swing2LimitDegrees=45
  swing2Motion=Free
  twistLimitDegrees=45
  twistMotion=Free
 }
 {bone1Name=dyn_skirt_f2_l_02, bone2Name=dyn_skirt_f2_l_01, swing1LimitDegrees=10, swing2LimitDegrees=20, twistLimitDegrees=5}
 {
  bone1Name=dyn_skirt_f2_l_02
  bone2Name=dyn_skirt_f_l_02
  swing1LimitDegrees=45
  swing1Motion=Free
  swing2LimitDegrees=45
  swing2Motion=Free
  twistLimitDegrees=45
  twistMotion=Free
 }
 {bone1Name=dyn_skirt_f2_l_03, bone2Name=dyn_skirt_f2_l_02, swing1LimitDegrees=10, swing2LimitDegrees=20, twistLimitDegrees=5}
 {
  bone1Name=dyn_skirt_f_l_03
  bone2Name=dyn_skirt_f2_l_03
  swing1LimitDegrees=45
  swing1Motion=Free
  swing2LimitDegrees=45
  swing2Motion=Free
  twistLimitDegrees=45
  twistMotion=Free
 }
 {
  bone1Name=dyn_skirt_f_c_03
  bone2Name=dyn_skirt_f2_l_03
  swing1LimitDegrees=45
  swing1Motion=Free
  swing2LimitDegrees=45
  swing2Motion=Free
  twistLimitDegrees=45
 }
 {
  bone1Name=dyn_skirt_f2_r_base
  bone2Name=pelvis
  swing1LimitDegrees=45
  swing1Motion=Free
  swing2LimitDegrees=45
  swing2Motion=Free
  twistLimitDegrees=45
  twistMotion=Free
 }
 {bone1Name=dyn_skirt_f2_r_01, bone2Name=dyn_skirt_f2_r_base, swing1LimitDegrees=10, swing2LimitDegrees=25, twistLimitDegrees=5}
 {
  bone1Name=dyn_skirt_f2_r_01
  bone2Name=dyn_skirt_f_r_01
  swing1LimitDegrees=45
  swing1Motion=Free
  swing2LimitDegrees=45
  swing2Motion=Free
  twistLimitDegrees=45
  twistMotion=Free
 }
 {bone1Name=dyn_skirt_f2_r_02, bone2Name=dyn_skirt_f2_r_01, swing1LimitDegrees=10, swing2LimitDegrees=25, twistLimitDegrees=5}
 {
  bone1Name=dyn_skirt_f_r_02
  bone2Name=dyn_skirt_f2_r_02
  swing1LimitDegrees=45
  swing1Motion=Free
  swing2LimitDegrees=45
  swing2Motion=Free
  twistLimitDegrees=45
  twistMotion=Free
 }
 {
  bone1Name=dyn_skirt_f2_r_02
  bone2Name=dyn_skirt_f_r_02
  swing1LimitDegrees=45
  swing1Motion=Free
  swing2LimitDegrees=45
  swing2Motion=Free
  twistLimitDegrees=45
  twistMotion=Free
 }
 {bone1Name=dyn_skirt_f2_r_03, bone2Name=dyn_skirt_f2_r_02, swing1LimitDegrees=10, swing2LimitDegrees=25, twistLimitDegrees=5}
 {
  bone1Name=dyn_skirt_f_c_03
  bone2Name=dyn_skirt_f2_r_03
  swing1LimitDegrees=45
  swing1Motion=Free
  swing2LimitDegrees=45
  swing2Motion=Free
  twistLimitDegrees=45
  twistMotion=Free
 }
 {
  bone1Name=dyn_skirt_f_r_03
  bone2Name=dyn_skirt_f2_r_03
  swing1LimitDegrees=45
  swing1Motion=Free
  swing2LimitDegrees=45
  swing2Motion=Free
  twistLimitDegrees=45
  twistMotion=Free
 }
 {
  bone1Name=dyn_skirt_f_c_base
  bone2Name=pelvis
  swing1LimitDegrees=45
  swing1Motion=Free
  swing2LimitDegrees=45
  swing2Motion=Free
  twistLimitDegrees=45
  twistMotion=Free
 }
 {bone1Name=dyn_skirt_f_c_01, bone2Name=dyn_skirt_f_c_base, swing1LimitDegrees=10, swing2LimitDegrees=25, twistLimitDegrees=5}
 {bone1Name=dyn_skirt_f_c_02, bone2Name=dyn_skirt_f_c_01, swing1LimitDegrees=10, swing2LimitDegrees=25, twistLimitDegrees=5}
 {bone1Name=dyn_skirt_f_c_03, bone2Name=dyn_skirt_f_c_02, swing1LimitDegrees=10, swing2LimitDegrees=25, twistLimitDegrees=5}
 {
  bone1Name=dyn_skirt_f_l_base
  bone2Name=pelvis
  swing1LimitDegrees=45
  swing1Motion=Free
  swing2LimitDegrees=45
  swing2Motion=Free
  twistLimitDegrees=45
  twistMotion=Free
 }
 {bone1Name=dyn_skirt_f_l_01, bone2Name=dyn_skirt_f_l_base, swing1LimitDegrees=10, swing2LimitDegrees=22.5, twistLimitDegrees=5}
 {
  bone1Name=dyn_skirt_s_l_01
  bone2Name=dyn_skirt_f_l_01
  swing1LimitDegrees=45
  swing1Motion=Free
  swing2LimitDegrees=45
  swing2Motion=Free
  twistLimitDegrees=45
  twistMotion=Free
 }
 {bone1Name=dyn_skirt_f_l_02, bone2Name=dyn_skirt_f_l_01, swing1LimitDegrees=10, swing2LimitDegrees=22.5, twistLimitDegrees=5}
 {
  bone1Name=dyn_skirt_s_l_02
  bone2Name=dyn_skirt_f_l_02
  swing1LimitDegrees=45
  swing1Motion=Free
  swing2LimitDegrees=45
  swing2Motion=Free
  twistLimitDegrees=45
  twistMotion=Free
 }
 {bone1Name=dyn_skirt_f_l_03, bone2Name=dyn_skirt_f_l_02, swing1LimitDegrees=10, swing2LimitDegrees=22.5, twistLimitDegrees=5}
 {
  bone1Name=dyn_skirt_f_r_base
  bone2Name=pelvis
  swing1LimitDegrees=45
  swing1Motion=Free
  swing2LimitDegrees=45
  swing2Motion=Free
  twistLimitDegrees=45
  twistMotion=Free
 }
 {bone1Name=dyn_skirt_f_r_01, bone2Name=dyn_skirt_f_r_base, swing1LimitDegrees=10, swing2LimitDegrees=25, twistLimitDegrees=5}
 {
  bone1Name=dyn_skirt_s_r_01
  bone2Name=dyn_skirt_f_r_01
  swing1LimitDegrees=45
  swing1Motion=Free
  swing2LimitDegrees=45
  swing2Motion=Free
  twistLimitDegrees=45
  twistMotion=Free
 }
 {bone1Name=dyn_skirt_f_r_02, bone2Name=dyn_skirt_f_r_01, swing1LimitDegrees=10, swing2LimitDegrees=25, twistLimitDegrees=5}
 {
  bone1Name=dyn_skirt_s_r_02
  bone2Name=dyn_skirt_f_r_02
  swing1LimitDegrees=45
  swing1Motion=Free
  swing2LimitDegrees=45
  swing2Motion=Free
  twistLimitDegrees=45
  twistMotion=Free
 }
 {bone1Name=dyn_skirt_f_r_03, bone2Name=dyn_skirt_f_r_02, swing1LimitDegrees=10, swing2LimitDegrees=25, twistLimitDegrees=5}
 {
  bone1Name=dyn_skirt_s_l_base
  bone2Name=pelvis
  swing1LimitDegrees=45
  swing1Motion=Free
  swing2LimitDegrees=45
  swing2Motion=Free
  twistLimitDegrees=45
  twistMotion=Free
 }
 {bone1Name=dyn_skirt_s_l_01, bone2Name=dyn_skirt_s_l_base, swing1LimitDegrees=15, swing2LimitDegrees=25, twistLimitDegrees=5}
 {bone1Name=dyn_skirt_s_l_02, bone2Name=dyn_skirt_s_l_01, swing1LimitDegrees=15, swing2LimitDegrees=25, twistLimitDegrees=5}
 {bone1Name=dyn_skirt_s_l_03, bone2Name=dyn_skirt_s_l_02, swing1LimitDegrees=15, swing2LimitDegrees=25, twistLimitDegrees=5}
 {
  bone1Name=dyn_skirt_s_r_base
  bone2Name=pelvis
  swing1LimitDegrees=45
  swing1Motion=Free
  swing2LimitDegrees=45
  swing2Motion=Free
  twistLimitDegrees=45
  twistMotion=Free
 }
 {bone1Name=dyn_skirt_s_r_01, bone2Name=dyn_skirt_s_r_base, swing1LimitDegrees=15, swing2LimitDegrees=25, twistLimitDegrees=5}
 {bone1Name=dyn_skirt_s_r_02, bone2Name=dyn_skirt_s_r_01, swing1LimitDegrees=15, swing2LimitDegrees=25, twistLimitDegrees=5}
 {bone1Name=dyn_skirt_s_r_03, bone2Name=dyn_skirt_s_r_02, swing1LimitDegrees=15, swing2LimitDegrees=25, twistLimitDegrees=5}
]
