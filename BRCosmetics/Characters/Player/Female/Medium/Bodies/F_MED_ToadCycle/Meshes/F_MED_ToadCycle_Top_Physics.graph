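# Physics asset for the top of the ToadCycle medium female player cosmetic, setting up a scarf with front and back dynamic chains.
# Tiny base capsules join each chain to spine_05 with free angular constraints, then five back and three front capsule links follow with limited swing and tight twist.
# Capsules on the upper spine and right arm, plus a box on the right upper arm, provide body shapes around the scarf.
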PhysicsAsset /BRCosmetics/Characters/Player/Female/Medium/Bodies/F_MED_ToadCycle/Meshes/F_MED_ToadCycle_Top_Physics
bodies {
 base_scarf_bk [{shapeType=Capsule, length=0.505, radius=0.505, shapeName=base_scarf_bk_capsule, rotation=<rot 0 -90 0>}]
 base_scarf_fr [{shapeType=Capsule, length=0.505, radius=0.505, shapeName=base_scarf_fr_capsule, rotation=<rot 0 -90 0>}]
 clavicle_r [{shapeType=Capsule, length=12, radius=6.5, shapeName=upperarm_r_capsule, center=<-5 0 3>, rotation=<rot -90 0 0>}]
 dyn_scarf_bk_01 [{shapeType=Capsule, length=6, radius=3, center=<5 0 0.5>, rotation=<rot -90 90 0>}]
 dyn_scarf_bk_02 [{shapeType=Capsule, length=6, radius=3.5, center=<5 0 1.5>, rotation=<rot -90 90 0>}]
 dyn_scarf_bk_03 [{shapeType=Capsule, length=6, radius=3.5, center=<5 0 2.5>, rotation=<rot -90 90 0>}]
 dyn_scarf_bk_04 [{shapeType=Capsule, length=6, radius=3, center=<5 0 2>, rotation=<rot -90 90 0>}]
 dyn_scarf_bk_05 [{shapeType=Capsule, length=6, radius=3, center=<5 0 1.5>, rotation=<rot -90 90 0>}]
 dyn_scarf_fr_01 [{shapeType=Capsule, length=6, radius=1.5, center=<6 -3 4>, rotation=<rot -90 0 0>}]
 dyn_scarf_fr_02 [
  {shapeType=Capsule, length=6, radius=2.5, center=<6 -3 3>, rotation=<rot -90 0 0>}
  {shapeType=Capsule, length=10, radius=1, center=<9.3896 0.121 -2.4135>, rotation=<rot 65 90.0005 0.0006>}
 ]
 dyn_scarf_fr_03 [{shapeType=Capsule, length=6, radius=2, center=<6 -3 2>, rotation=<rot -90 0 0>}]
 lowerarm_r [{shapeType=Capsule, length=12, radius=3, shapeName=upperarm_r_capsule, center=<-12 0 0>, rotation=<rot -90 0 0>}]
 spine_02 [{shapeType=Capsule, length=9, radius=10, shapeName=spine_02_capsule, center=<0 1.4449 0.3596>}]
 spine_03 [{shapeType=Capsule, length=9, radius=10, shapeName=spine_03_capsule, center=<0 1.7971 0>}]
 spine_04 [{shapeType=Capsule, length=8, radius=10.8882, shapeName=spine_04_capsule, center=<-0.9375 2.1315 0>}]
 spine_05 [{shapeType=Capsule, length=9, radius=10.5, shapeName=spine_05_capsule, center=<-2.5484 1.5107 0>}]
 upperarm_r [
  {shapeType=Capsule, length=12, radius=5, shapeName=upperarm_r_capsule, center=<-12 3.7359 0>, rotation=<rot -90 0 0>}
  {shapeType=Capsule, length=12, radius=3, shapeName=upperarm_r_capsule, center=<-12 0 0>, rotation=<rot -90 0 0>}
  {
   shapeType=Box
   extentX=35.5
   extentY=2.5
   extentZ=11.5
   shapeName=upperarm_r_box
   center=<-17.8668 0 -2.8935>
   rotation=<rot 0 0.0002 89.9998>
  }
 ]
}
constraints [
 {
  bone1Name=base_scarf_bk
  bone2Name=spine_05
  swing1LimitDegrees=45
  swing1Motion=Free
  swing2LimitDegrees=45
  swing2Motion=Free
  twistLimitDegrees=45
  twistMotion=Free
 }
 {bone1Name=dyn_scarf_bk_01, bone2Name=base_scarf_bk, swing1LimitDegrees=25, swing2LimitDegrees=25, twistLimitDegrees=2}
 {bone1Name=dyn_scarf_bk_02, bone2Name=dyn_scarf_bk_01, swing1LimitDegrees=25, swing2LimitDegrees=20, twistLimitDegrees=2}
 {bone1Name=dyn_scarf_bk_03, bone2Name=dyn_scarf_bk_02, swing1LimitDegrees=25, swing2LimitDegrees=15, twistLimitDegrees=2}
 {bone1Name=dyn_scarf_bk_04, bone2Name=dyn_scarf_bk_03, swing1LimitDegrees=25, swing2LimitDegrees=25, twistLimitDegrees=2}
 {bone1Name=dyn_scarf_bk_05, bone2Name=dyn_scarf_bk_04, swing1LimitDegrees=25, swing2LimitDegrees=25, twistLimitDegrees=2}
 {
  bone1Name=base_scarf_fr
  bone2Name=spine_05
  swing1LimitDegrees=45
  swing1Motion=Free
  swing2LimitDegrees=45
  swing2Motion=Free
  twistLimitDegrees=45
  twistMotion=Free
 }
 {bone1Name=dyn_scarf_fr_01, bone2Name=base_scarf_fr, swing1LimitDegrees=25, swing2LimitDegrees=25, twistLimitDegrees=5}
 {bone1Name=dyn_scarf_fr_02, bone2Name=dyn_scarf_fr_01, swing1LimitDegrees=25, swing2LimitDegrees=25, twistLimitDegrees=5}
 {bone1Name=dyn_scarf_fr_03, bone2Name=dyn_scarf_fr_02, swing1LimitDegrees=25, swing2LimitDegrees=25, twistLimitDegrees=5}
 {
  bone1Name=clavicle_r
  bone2Name=spine_05
  swing1LimitDegrees=45
  swing1Motion=Free
  swing2LimitDegrees=45
  swing2Motion=Free
  twistLimitDegrees=45
  twistMotion=Free
 }
]
