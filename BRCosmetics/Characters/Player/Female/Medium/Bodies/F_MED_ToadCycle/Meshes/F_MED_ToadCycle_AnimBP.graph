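# Cosmetic AnimBP for the female medium ToadCycle body, animating coat skirt and scarf secondary motion over the copied mesh pose.
# A Control Rig feeds pelvis-based RigidBody coat physics, plus a spine-based top simulation used when the scarf meta tag is present.
# Backbling or cape selects a back scarf and skirt constraint branch, followed by front scarf and skirt adjustments.
# The EventGraph updates gravity override, item flags, and constraint alphas for crouch, emote and melee.
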AnimBlueprint /BRCosmetics/Characters/Player/Female/Medium/Bodies/F_MED_ToadCycle/Meshes/F_MED_ToadCycle_AnimBP
target_skeleton=&/F_MED_ToadCycle_Skeleton
graphs [
 AnimGraph {nodes [
   "Transform (Modify) Bone"!(ComponentPose, bAlphaBoolEnabled:bool=True, Translation="0,0,0", Rotation="0,0,0", Scale="1,1,1") -> (Pose) {alphaInputType=Bool, boneToModify=~}
   n1 = "Transform (Modify) Bone - Bone: base_scarf_bk"!(
    ComponentPose=Local To Component(LocalPose=Use cached pose 'Physics_cache'())
    bAlphaBoolEnabled:bool=True
    Translation="5,0,0"
    Rotation="0,-30,0"
    Scale="1,1,1"
   ) -> (Pose) {
    alphaInputType=Bool
    rotationMode=BMM_Additive
    rotationSpace=BCS_BoneSpace
    scaleMode=BMM_Additive
    scaleSpace=BCS_BoneSpace
    translationMode=BMM_Additive
    translationSpace=BCS_BoneSpace
    alphaBoolBlend {blendInTime=0.2, blendOutTime=0.2}
    boneToModify {boneName=base_scarf_bk}
    rotation {yaw=-30}
    translation {x=5}
   }
   n2 = Component To Local(ComponentPose=@n1)
   n3 = "Constraint - dyn_scarf_bk_01 (P)"(ComponentPose=Local To Component(LocalPose=Use cached pose 'Physics_cache'()), Alpha=1, ConstraintWeights_0=1) {boneToModify {boneName=dyn_scarf_bk_01}, constraintSetup [{targetBone {boneName=clavicle_r}}], constraintWeights [1]}
   n4 = "Constraint - dyn_scarf_bk_02 (P)"(ComponentPose=@n3, Alpha=1, ConstraintWeights_0=1) {boneToModify {boneName=dyn_scarf_bk_02}, constraintSetup [{targetBone {boneName=clavicle_r}}], constraintWeights [1]}
   n5 = "Constraint - dyn_scarf_bk_03 (P)"(ComponentPose=@n4, Alpha=1, ConstraintWeights_0=1) {boneToModify {boneName=dyn_scarf_bk_03}, constraintSetup [{targetBone {boneName=clavicle_r}}], constraintWeights [1]}
   n6 = "Constraint - dyn_scarf_bk_04 (P)"(ComponentPose=@n5, Alpha=1, ConstraintWeights_0=1) {boneToModify {boneName=dyn_scarf_bk_04}, constraintSetup [{targetBone {boneName=spine_03}}], constraintWeights [1]}
   n7 = "Constraint - dyn_scarf_bk_05 (P)"(ComponentPose=@n6, Alpha=1, ConstraintWeights_0=1) {boneToModify {boneName=dyn_scarf_bk_05}, constraintSetup [{targetBone {boneName=spine_04}}], constraintWeights [1]}
   n8 = "Constraint - dyn_scarf_bk_06 (P)"(ComponentPose=@n7, Alpha=1, ConstraintWeights_0=1) {boneToModify {boneName=dyn_scarf_bk_06}, constraintSetup [{targetBone {boneName=spine_03}}], constraintWeights [1]}
   n9 = "Transform (Modify) Bone - Bone: dyn_scarf_bk_04"(ComponentPose=@n8, Alpha=1, Translation="0,0,0", Rotation="-15,-5,-25", Scale="1,1,1") {
    rotationMode=BMM_Additive
    rotationSpace=BCS_BoneSpace
    translationMode=BMM_Additive
    translationSpace=BCS_BoneSpace
    boneToModify {boneName=dyn_scarf_bk_04}
    rotation=<rot -15 -5 -25>
   }
   n10 = Get Alpha value constraints() -> (Alpha value constraints:double)
   n11 = "Constraint - dyn_skirt_s_l_01 (R)"(ComponentPose=@n9, Alpha=@n10, ConstraintWeights_0=1) {boneToModify {boneName=dyn_skirt_s_l_01}, constraintSetup [{transformType=Rotation, targetBone {boneName=thigh_l}}], constraintWeights [1]}
   n12 = "Constraint - dyn_skirt_b_l_01 (R)"(ComponentPose=@n11, Alpha=@n10, ConstraintWeights_0=1) {boneToModify {boneName=dyn_skirt_b_l_01}, constraintSetup [{transformType=Rotation, targetBone {boneName=thigh_l}}], constraintWeights [1]}
   n13 = "Constraint - dyn_skirt_b_r_01 (R)"(ComponentPose=@n12, Alpha=@n10, ConstraintWeights_0=1) {boneToModify {boneName=dyn_skirt_b_r_01}, constraintSetup [{transformType=Rotation, targetBone {boneName=thigh_r}}], constraintWeights [1]}
   n14 = "Constraint - dyn_skirt_s_r_01 (R)"(ComponentPose=@n13, Alpha=@n10, ConstraintWeights_0=1) {boneToModify {boneName=dyn_skirt_s_r_01}, constraintSetup [{transformType=Rotation, targetBone {boneName=thigh_r}}], constraintWeights [1]}
   n15 = "Constraint - dyn_skirt_s_l_02 (R)"(ComponentPose=@n14, Alpha=@n10, ConstraintWeights_0=1) {boneToModify {boneName=dyn_skirt_s_l_02}, constraintSetup [{transformType=Rotation, targetBone {boneName=thigh_l}}], constraintWeights [1]}
   n16 = "Constraint - dyn_skirt_b_l_02 (R)"(ComponentPose=@n15, Alpha=@n10, ConstraintWeights_0=1) {boneToModify {boneName=dyn_skirt_b_l_02}, constraintSetup [{transformType=Rotation, targetBone {boneName=thigh_l}}], constraintWeights [1]}
   n17 = "Constraint - dyn_skirt_b_r_02 (R)"(ComponentPose=@n16, Alpha=@n10, ConstraintWeights_0=1) {boneToModify {boneName=dyn_skirt_b_r_02}, constraintSetup [{transformType=Rotation, targetBone {boneName=thigh_r}}], constraintWeights [1]}
   n18 = "Constraint - dyn_skirt_s_r_02 (R)"(ComponentPose=@n17, Alpha=@n10, ConstraintWeights_0=1) {boneToModify {boneName=dyn_skirt_s_r_02}, constraintSetup [{transformType=Rotation, targetBone {boneName=thigh_r}}], constraintWeights [1]}
   n19 = "Constraint - dyn_skirt_s_l_03 (R)"(ComponentPose=@n18, Alpha=@n10, ConstraintWeights_0=1) {boneToModify {boneName=dyn_skirt_s_l_03}, constraintSetup [{transformType=Rotation, targetBone {boneName=thigh_l}}], constraintWeights [1]}
   n20 = "Constraint - dyn_skirt_b_l_03 (R)"(ComponentPose=@n19, Alpha=@n10, ConstraintWeights_0=1) {boneToModify {boneName=dyn_skirt_b_l_03}, constraintSetup [{transformType=Rotation, targetBone {boneName=thigh_l}}], constraintWeights [1]}
   n21 = "Constraint - dyn_skirt_b_r_03 (R)"(ComponentPose=@n20, Alpha=@n10, ConstraintWeights_0=1) {boneToModify {boneName=dyn_skirt_b_r_03}, constraintSetup [{transformType=Rotation, targetBone {boneName=thigh_r}}], constraintWeights [1]}
   n22 = "Constraint - dyn_skirt_s_r_03 (R)"(ComponentPose=@n21, Alpha=@n10, ConstraintWeights_0=1) {boneToModify {boneName=dyn_skirt_s_r_03}, constraintSetup [{transformType=Rotation, targetBone {boneName=thigh_r}}], constraintWeights [1]}
   n23 = "Constraint - dyn_skirt_s_l_04 (R)"(ComponentPose=@n22, Alpha=@n10, ConstraintWeights_0=1) {boneToModify {boneName=dyn_skirt_s_l_04}, constraintSetup [{transformType=Rotation, targetBone {boneName=thigh_l}}], constraintWeights [1]}
   n24 = "Constraint - dyn_skirt_b_l_04 (R)"(ComponentPose=@n23, Alpha=@n10, ConstraintWeights_0=1) {boneToModify {boneName=dyn_skirt_b_l_04}, constraintSetup [{transformType=Rotation, targetBone {boneName=thigh_l}}], constraintWeights [1]}
   n25 = "Constraint - dyn_skirt_b_r_04 (R)"(ComponentPose=@n24, Alpha=@n10, ConstraintWeights_0=1) {boneToModify {boneName=dyn_skirt_b_r_04}, constraintSetup [{transformType=Rotation, targetBone {boneName=thigh_r}}], constraintWeights [1]}
   n26 = "Constraint - dyn_skirt_s_r_04 (R)"(ComponentPose=@n25, Alpha=@n10, ConstraintWeights_0=1) {boneToModify {boneName=dyn_skirt_s_r_04}, constraintSetup [{transformType=Rotation, targetBone {boneName=thigh_r}}], constraintWeights [1]}
   n27 = "Transform (Modify) Bone - Bone: attach_backpack"!(ComponentPose=@n26, bAlphaBoolEnabled:bool=True, Translation="0,-1.2,0", Rotation="0,0,0", Scale="1,1,1") -> (Pose) {
    alphaInputType=Bool
    rotationMode=BMM_Additive
    rotationSpace=BCS_BoneSpace
    translationMode=BMM_Additive
    translationSpace=BCS_BoneSpace
    boneToModify {boneName=attach_backpack}
    translation {y=-1.2}
   }
   n28 = Component To Local(ComponentPose=@n27)
   n29 = OR Boolean(A=Get isWearingBackbling(), B=Get Is Wearing Cape() -> (Is Wearing Cape:bool))
   n30 = "Transform (Modify) Bone - Bone: base_scarf_fr"!(
    ComponentPose=Local To Component(LocalPose=Blend Poses by bool(BlendPose_0=@n28, BlendPose_1=@n2, bActiveValue=@n29) {blendType=Linear})
    bAlphaBoolEnabled:bool=True
    Translation="0,0,2"
    Rotation="0,0,-5"
    Scale="1,1,1"
   ) -> (Pose) {
    alphaInputType=Bool
    rotationMode=BMM_Additive
    rotationSpace=BCS_BoneSpace
    translationMode=BMM_Additive
    translationSpace=BCS_BoneSpace
    alphaBoolBlend {blendInTime=0.2, blendOutTime=0.2}
    boneToModify {boneName=base_scarf_fr}
    rotation {roll=-5}
    translation {z=2}
   }
   n31 = "Constraint - dyn_scarf_fr_01 (P)"(ComponentPose=@n30, Alpha=Get floatAlphaPEmote() -> (floatAlphaPEmote:double), ConstraintWeights_0=1) {
    alphaBoolBlend {blendInTime=0.2, blendOutTime=0.2}
    boneToModify {boneName=dyn_scarf_fr_01}
    constraintSetup [{targetBone {boneName=clavicle_r}}]
    constraintWeights [1]
   }
   n32 = "Transform (Modify) Bone - Bone: dyn_skirt_f_r_base"!(ComponentPose=@n31, bAlphaBoolEnabled:bool=True, Translation="0,0,0", Rotation="5,0,0", Scale="1,1,1") -> (Pose) {
    alphaInputType=Bool
    rotationMode=BMM_Additive
    rotationSpace=BCS_BoneSpace
    translationMode=BMM_Additive
    translationSpace=BCS_BoneSpace
    alphaBoolBlend {blendInTime=0.2, blendOutTime=0.2}
    boneToModify {boneName=dyn_skirt_f_r_base}
    rotation {pitch=5}
   }
   n33 = "Transform (Modify) Bone - Bone: dyn_skirt_f_l_base"!(ComponentPose=@n32, bAlphaBoolEnabled:bool=True, Translation="0,0,0", Rotation="5,0,0", Scale="1,1,1") -> (Pose) {
    alphaInputType=Bool
    rotationMode=BMM_Additive
    rotationSpace=BCS_BoneSpace
    translationMode=BMM_Additive
    translationSpace=BCS_BoneSpace
    alphaBoolBlend {blendInTime=0.2, blendOutTime=0.2}
    boneToModify {boneName=dyn_skirt_f_l_base}
    rotation {pitch=5}
   }
   n34 = "Transform (Modify) Bone - Bone: dyn_skirt_f2_r_base"!(ComponentPose=@n33, bAlphaBoolEnabled:bool=True, Translation="0,0,0", Rotation="5,0,0", Scale="1,1,1") -> (Pose) {
    alphaInputType=Bool
    rotationMode=BMM_Additive
    rotationSpace=BCS_BoneSpace
    translationMode=BMM_Additive
    translationSpace=BCS_BoneSpace
    alphaBoolBlend {blendInTime=0.2, blendOutTime=0.2}
    boneToModify {boneName=dyn_skirt_f2_r_base}
    rotation {pitch=5}
   }
   n35 = "Transform (Modify) Bone - Bone: dyn_skirt_f2_l_base"!(ComponentPose=@n34, bAlphaBoolEnabled:bool=True, Translation="0,0,0", Rotation="5,0,0", Scale="1,1,1") -> (Pose) {
    alphaInputType=Bool
    rotationMode=BMM_Additive
    rotationSpace=BCS_BoneSpace
    translationMode=BMM_Additive
    translationSpace=BCS_BoneSpace
    alphaBoolBlend {blendInTime=0.2, blendOutTime=0.2}
    boneToModify {boneName=dyn_skirt_f2_l_base}
    rotation {pitch=5}
   }
   n36 = "Constraint - dyn_skirt_f_l_01 (P)"!(ComponentPose=@n35, bAlphaBoolEnabled:bool=True, ConstraintWeights_0=1) -> (Pose) {
    alphaInputType=Bool
    alphaBoolBlend {blendInTime=0.2, blendOutTime=0.2}
    boneToModify {boneName=dyn_skirt_f_l_01}
    constraintSetup [{targetBone {boneName=thigh_twist_01_l}}]
    constraintWeights [1]
   }
   n37 = "Constraint - dyn_skirt_f2_l_01 (P)"!(ComponentPose=@n36, bAlphaBoolEnabled:bool=True, ConstraintWeights_0=1) -> (Pose) {
    alphaInputType=Bool
    alphaBoolBlend {blendInTime=0.2, blendOutTime=0.2}
    boneToModify {boneName=dyn_skirt_f2_l_01}
    constraintSetup [{targetBone {boneName=thigh_twist_01_l}}]
    constraintWeights [1]
   }
   n38 = "Constraint - dyn_skirt_f_r_01 (P)"(ComponentPose=@n37, Alpha=0.5, ConstraintWeights_0=1) {
    alphaScaleBiasClamp {scale=0.5}
    boneToModify {boneName=dyn_skirt_f_r_01}
    constraintSetup [{targetBone {boneName=thigh_twist_01_r}}]
    constraintWeights [1]
   }
   n39 = "Constraint - dyn_skirt_f2_r_01 (P)"!(ComponentPose=@n38, bAlphaBoolEnabled:bool=True, ConstraintWeights_0=1) -> (Pose) {
    alphaInputType=Bool
    alphaBoolBlend {blendInTime=0.2, blendOutTime=0.2}
    boneToModify {boneName=dyn_skirt_f2_r_01}
    constraintSetup [{targetBone {boneName=thigh_twist_01_r}}]
    constraintWeights [1]
   }
   n40 = "Constraint - dyn_skirt_f2_l_03 (P)"!(ComponentPose=@n39, bAlphaBoolEnabled:bool=True, ConstraintWeights_0=1) -> (Pose) {
    alphaInputType=Bool
    alphaBoolBlend {blendInTime=0.2, blendOutTime=0.2}
    boneToModify {boneName=dyn_skirt_f2_l_03}
    constraintSetup [{targetBone {boneName=thigh_twist_01_l}}]
    constraintWeights [1]
   }
   n41 = "Constraint - dyn_skirt_f2_r_03 (P)"!(ComponentPose=@n40, bAlphaBoolEnabled:bool=True, ConstraintWeights_0=1) -> (Pose) {
    alphaInputType=Bool
    alphaBoolBlend {blendInTime=0.2, blendOutTime=0.2}
    boneToModify {boneName=dyn_skirt_f2_r_03}
    constraintSetup [{targetBone {boneName=thigh_twist_01_r}}]
    constraintWeights [1]
   }
   n42 = "Constraint - dyn_skirt_f_l_03 (P)"!(ComponentPose=@n41, bAlphaBoolEnabled:bool=True, ConstraintWeights_0=1) -> (Pose) {
    alphaInputType=Bool
    alphaBoolBlend {blendInTime=0.2, blendOutTime=0.2}
    boneToModify {boneName=dyn_skirt_f_l_03}
    constraintSetup [{targetBone {boneName=thigh_twist_01_l}}]
    constraintWeights [1]
   }
   n43 = "Constraint - dyn_skirt_f_r_03 (P)"(ComponentPose=@n42, Alpha=0.5, ConstraintWeights_0=1) {
    alphaScaleBiasClamp {scale=0.5}
    boneToModify {boneName=dyn_skirt_f_r_03}
    constraintSetup [{targetBone {boneName=thigh_twist_01_r}}]
    constraintWeights [1]
   }
   n44 = "Transform (Modify) Bone - Bone: base_scarf_fr"(ComponentPose=@n43, bAlphaBoolEnabled=True, Translation="0,0,1", Rotation="5,0,0", Scale="1,1,1") {
    alphaInputType=Bool
    rotationMode=BMM_Additive
    rotationSpace=BCS_BoneSpace
    translationMode=BMM_Additive
    translationSpace=BCS_BoneSpace
    alphaBoolBlend {blendInTime=0.2, blendOutTime=0.2}
    boneToModify {boneName=base_scarf_fr}
    rotation {pitch=5}
    translation {z=1}
   }
   Output Pose(Result=Component To Local(ComponentPose=@n44)) {layerGroup=None}
   Save cached pose 'Physics_cache'(Pose=Blend Poses by bool(
     BlendPose_0=Use cached pose 'ScarfPhysics_cache'()
     BlendPose_1=Use cached pose 'CoatPhysics_Cache'()
     bActiveValue=Get scarf_item() -> (scarf_item:bool)
    ) {blendType=Linear}) {cacheName=Physics_cache}
   n45 = Control Rig!(Source=Copy Pose From Mesh() {bUseAttachedParent=true}, Alpha, isSliding:bool=Get bIsSliding()) -> (Pose) {controlRigAssetReference {blueprintRigClass=&/F_MED_ToadCycle_CtrlRig.F_MED_ToadCycle_CtrlRig_C}}
   n46 = RigidBody!(ComponentPose=Local To Component(LocalPose=@n45), Alpha, OverrideWorldGravity:Vector=Get GravityOverride(), ExternalForce) -> (Pose) {
    bClampLinearTranslationLimitToRefPose=true
    bOverrideWorldGravity=true
    overridePhysicsAsset=&/F_MED_ToadCycle_Physics
    simulationSpace=BaseBoneSpace
    alphaBoolBlend {blendOption=Linear}
    baseBoneRef {boneName=pelvis}
    overrideWorldGravity {z=-980}
    simSpaceSettings {
     dampingAlpha=0
     maxAngularAcceleration=40
     maxAngularVelocity=10
     maxLinearAcceleration=1200
     maxLinearVelocity=600
     velocityScaleZ=0.5
     worldAlpha=1
     externalLinearDragV=<0.3 0.3 0.3>
    }
   }
   Save cached pose 'CoatPhysics_Cache'(Pose=Component To Local(ComponentPose=@n46)) {cacheName=CoatPhysics_Cache}
   n47 = RigidBody(ComponentPose=Local To Component(LocalPose=Use cached pose 'CoatPhysics_Cache'()), OverrideWorldGravity=Get GravityOverride()) {
    bClampLinearTranslationLimitToRefPose=true
    bOverrideWorldGravity=true
    overridePhysicsAsset=&/F_MED_ToadCycle_Top_Physics
    simulationSpace=BaseBoneSpace
    alphaBoolBlend {blendOption=Linear}
    baseBoneRef {boneName=spine_05}
    overrideWorldGravity {z=-980}
    simSpaceSettings {
     dampingAlpha=0
     maxAngularAcceleration=28
     maxAngularVelocity=7
     maxLinearAcceleration=900
     maxLinearVelocity=600
     velocityScaleZ=0.4
     worldAlpha=1
     externalLinearDragV=<0.3 0.3 0.3>
    }
   }
   Save cached pose 'ScarfPhysics_cache'(Pose=Component To Local(ComponentPose=@n47)) {cacheName=ScarfPhysics_cache}
  ]}
 EventGraph {nodes [
   Event BlueprintUpdateAnimation()
   > n1 = Sequence!(execute) -> (then_0, then_1, then_2:Exec)
   > n2 = "Gravity Override State Function (Parameter Asset) (Nativized)"(
    CharacterPartAnimInstance=Self-Reference()
    GameTimeInSeconds=GetGameTimeInSeconds()
    Parameters=&/F_MED_ToadCycle_Parameters
   )
   > Set GravityOverride(GravityOverride=@n2)
   > Set RBAN Settings Out(RBAN Settings Out=@n2.OutRigidBodyAnimNodeParameters)
   > n3 = Branch(Condition=OR Boolean(A=Get bIsPlayingEmote(), B=Get bIsPlayingMeleeAnim()))
   > Set floatAlphaPEmote(floatAlphaPEmote:double=0.5) -> (Output_Get:double)
   Set floatAlphaPEmote(execute=@n3.else, floatAlphaPEmote=0)
   n4 = Branch(execute=@n1.then_2, Condition=CharacterItemDefinitionHasAnyMetaTag(Tags=Make Array("[0]"=(TagName="Cosmetics.MetaTags.ToadCycleSarf"))))
   > Set scarf_item(scarf_item:bool=true) -> (Output_Get:bool)
   Set scarf_item(execute=@n4.else, scarf_item=false)
   n5 = CheckForBackbling()
   Set isWearingBackbling(
    execute=@n1.then_1
    isWearingBackbling=AND Boolean!(A=@n5, B=NOT Boolean(A=@n5.bIsBackblingACape), C:bool=NOT Boolean(A=@n5.bIsBackblingLong)) -> (ReturnValue)
   )
   > Set Is Wearing Cape(Is Wearing Cape=OR Boolean(A=@n5.bIsBackblingACape, B=@n5.bIsBackblingLong))
   > n6 = Branch(Condition=Get bIsCrouching())
   > Set Alpha value constraints(Alpha value constraints:double=0.02) -> (Output_Get:double)
   Set Alpha value constraints(execute=@n6.else, Alpha value constraints=0.2)
  ]}
]
variables [
 Alpha value constraints {type=number}
 Character_ToadCycle_B {type=AthenaCosmeticItemDefinition, default=null}
 floatAlphaPEmote {type=number}
 GravityOverride {type=Vector}
 Is Wearing Cape {type=boolean}
 isWearingBackbling {type=boolean}
 RBAN Settings Out {type=FortRigidBodyAnimNodeParameters}
 scarf_item {type=boolean}
]
