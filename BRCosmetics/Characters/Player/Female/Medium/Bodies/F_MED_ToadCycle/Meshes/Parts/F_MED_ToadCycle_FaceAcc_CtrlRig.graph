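# Control Rig for the F_MED_ToadCycle face accessory that procedurally adjusts back-hair bones as the head and arms move.
# Sequenced left and right chains parent-constrain hair bones between head and centre strand, then spherical pose readers weight corrective constraints and offsets.
# Upper-arm readers adjust side and centre hair; head-versus-clavicle readers add asymmetric corrections per side.
# The centre 03a constraint chain has no link to execution.
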ControlRigBlueprint /BRCosmetics/Characters/Player/Female/Medium/Bodies/F_MED_ToadCycle/Meshes/Parts/F_MED_ToadCycle_FaceAcc_CtrlRig
graphs [
 ContainedGraph {nodes [
   n1 = RigVMFunctionEntryNode()
   SequenceAggregate = &RigVMFunction_Sequence(ExecuteContext=@n1.ExecuteContext)
   SequenceAggregate_1 = &RigVMFunction_Sequence(ExecuteContext=@SequenceAggregate.B)
   SequenceAggregate_2 = &RigVMFunction_Sequence(ExecuteContext=@SequenceAggregate_1.B)
   RigVMFunctionReturnNode(A=@SequenceAggregate.A, B=@SequenceAggregate_1.A, C=@SequenceAggregate_2.A, D=@SequenceAggregate_2.B)
  ]}
 RigVMFunctionLibrary {links [], nodes []}
 RigVMModel {nodes [
   &RigUnit_ParentConstraint(
    AdvancedSettings=(RotationOrderForFilter=XZY)
    Child=(Type=Bone,Name="dyn_Hair_Back_03a_01")
    ExecutePin=()
    Filter=(TranslationFilter=(bX=true,bY=true,bZ=true),RotationFilter=(bX=true,bY=true,bZ=true),ScaleFilter=(bX=True,bY=True,bZ=True))
    Parents=((Item=(Type=Bone,Name="dyn_Hair_Back_03"),Weight=0.8),(Item=(Type=Bone,Name="head"),Weight=0.2))
    Weight=1
   )
   > &RigUnit_ParentConstraint(
    AdvancedSettings=(RotationOrderForFilter=XZY)
    Child=(Type=Bone,Name="dyn_Hair_Back_03a_02")
    Filter=(TranslationFilter=(bX=True,bY=True,bZ=True),RotationFilter=(bX=true,bY=true,bZ=true),ScaleFilter=(bX=True,bY=True,bZ=True))
    Parents=((Item=(Type=Bone,Name="dyn_Hair_Back_03"),Weight=0.9),(Item=(Type=Bone,Name="head"),Weight=0.1))
    Weight=1
   )
   > &RigUnit_ParentConstraint(
    AdvancedSettings=(RotationOrderForFilter=XZY)
    Child=(Type=Bone,Name="dyn_Hair_Back_03a_03")
    Filter=(TranslationFilter=(bX=True,bY=True,bZ=True),RotationFilter=(bX=true,bY=true,bZ=true),ScaleFilter=(bX=True,bY=True,bZ=True))
    Parents=((Item=(Type=Bone,Name="dyn_Hair_Back_03"),Weight=1),(Item=(Type=Bone,Name="dyn_Hair_Back_04"),Weight=0))
    Weight=1
   )
   > &RigUnit_ParentConstraint(
    AdvancedSettings=(RotationOrderForFilter=XZY)
    Child=(Type=Bone,Name="dyn_Hair_Back_03a_04")
    Filter=(TranslationFilter=(bX=true,bY=true,bZ=true),RotationFilter=(bX=true,bY=true,bZ=true),ScaleFilter=(bX=True,bY=True,bZ=True))
    Parents=((Item=(Type=Bone,Name="dyn_Hair_Back_03"),Weight=0.5),(Item=(Type=Bone,Name="dyn_Hair_Back_04"),Weight=0.5))
    Weight=1
   )
   RigUnit_BeginExecution = &RigUnit_BeginExecution()
   SequenceAggregate = RigVMAggregateNode(ExecuteContext=@RigUnit_BeginExecution.ExecutePin)
   &RigUnit_ParentConstraint(
    AdvancedSettings=(RotationOrderForFilter=XZY)
    Child=(Type=Bone,Name="dyn_Hair_BackL_01")
    Filter=(TranslationFilter=(bX=true,bY=true,bZ=true),RotationFilter=(bX=true,bY=true,bZ=true),ScaleFilter=(bX=True,bY=True,bZ=True))
    Parents=((Item=(Type=Bone,Name="dyn_Hair_Back_01"),Weight=0),(Item=(Type=Bone,Name="head"),Weight=1))
    Weight=1
    ExecutePin=@SequenceAggregate.B
   )
   > &RigUnit_ParentConstraint(
    AdvancedSettings=(RotationOrderForFilter=XZY)
    Child=(Type=Bone,Name="dyn_Hair_BackL_02")
    Filter=(TranslationFilter=(bX=True,bY=True,bZ=True),RotationFilter=(bX=true,bY=true,bZ=true),ScaleFilter=(bX=True,bY=True,bZ=True))
    Parents=((Item=(Type=Bone,Name="dyn_Hair_Back_02"),Weight=0),(Item=(Type=Bone,Name="head"),Weight=1))
    Weight=1
   )
   > &RigUnit_ParentConstraint(
    AdvancedSettings=(RotationOrderForFilter=XZY)
    Child=(Type=Bone,Name="dyn_Hair_BackL_03")
    Filter=(TranslationFilter=(bX=True,bY=True,bZ=True),RotationFilter=(bX=true,bY=true,bZ=true),ScaleFilter=(bX=True,bY=True,bZ=True))
    Parents=((Item=(Type=Bone,Name="dyn_Hair_Back_03"),Weight=0.3),(Item=(Type=Bone,Name="head"),Weight=0.7))
    Weight=1
   )
   > &RigUnit_ParentConstraint(
    AdvancedSettings=(RotationOrderForFilter=XZY)
    Child=(Type=Bone,Name="dyn_Hair_BackL_04")
    Filter=(TranslationFilter=(bX=True,bY=True,bZ=True),RotationFilter=(bX=true,bY=true,bZ=true),ScaleFilter=(bX=True,bY=True,bZ=True))
    Parents=((Item=(Type=Bone,Name="dyn_Hair_Back_04"),Weight=1),(Item=(Type=Bone,Name="spine_05"),Weight=0))
    Weight=1
   )
   > n1 = &RigUnit_SphericalPoseReader(
    ActiveRegionScaleFactors=(PositiveWidth=0.6,NegativeWidth=0.6,PositiveHeight=0.6,NegativeHeight=0.6)
    ActiveRegionSize=0.1
    Debug=(bDrawDebug=false,DebugScale=25,DebugSegments=20,DebugThickness=0.25)
    DriverAxis=(X=1,Y=0,Z=0)
    DriverItem=(Type=Bone,Name="upperarm_l")
    FalloffRegionScaleFactors=(PositiveWidth=1,NegativeWidth=0.6,PositiveHeight=0.6,NegativeHeight=0.6)
    FalloffSize=0.55
    FlipHeightScaling=false
    FlipWidthScaling=false
    OptionalParentItem=(Type=Bone,Name="spine_05")
    RotationOffset=(X=-80,Y=-30,Z=-10)
   )
   > &RigUnit_ParentConstraint(
    AdvancedSettings=(RotationOrderForFilter=XZY)
    Child=(Type=Bone,Name="dyn_Hair_BackL_03")
    Filter=(TranslationFilter=(bX=false,bY=false,bZ=false),RotationFilter=(bX=true,bY=true,bZ=true),ScaleFilter=(bX=True,bY=True,bZ=True))
    Parents=((Item=(Type=Bone,Name="dyn_Hair_Back_03"),Weight=0.3),(Item=(Type=Bone,Name="upperarm_l"),Weight=0.7))
    Weight=@n1.OutputParam
   )
   n2 = &RigVMFunction_MathQuaternionFromEuler(Euler=(X=20,Y=-10,Z=0), RotationOrder=YXZ)
   > &RigUnit_ModifyTransforms(
    ItemToModify="((Item=(Type=Bone,Name=\"dyn_Hair_BackL_02\"),Transform=(Rotation=(X=0,Y=0,Z=0,W=1),Translation=(X=2,Y=-1,Z=1),Scale3D=(X=1,Y=1,Z=1))),(Item=(Type=Bone,Name=\"dyn_Hair_BackL_04\"),Transform=(Rotation=(X=0,Y=0,Z=0,W=1),Translation=(X=0,Y=0,Z=0),Scale3D=(X=1,Y=1 ...(truncated)"
    ItemToModify.0.Transform.Rotation=@n2.Result
    ItemToModify.1.Transform.Rotation=@n2.Result
    Weight=@n1.OutputParam
   )
   n3 = &RigVMFunction_MathQuaternionFromEuler(Euler=(X=0,Y=0,Z=5))
   > &RigUnit_ModifyTransforms(
    ItemToModify="((Item=(Type=Bone,Name=\"dyn_Hair_Back_03\"),Transform=(Rotation=(X=0,Y=0,Z=0,W=1),Translation=(X=0,Y=0,Z=1),Scale3D=(X=1,Y=1,Z=1))),(Item=(Type=Bone,Name=\"dyn_Hair_Back_04\"),Transform=(Rotation=(X=0,Y=0,Z=0,W=1),Translation=(X=0,Y=0,Z=0),Scale3D=(X=1,Y=1,Z= ...(truncated)"
    ItemToModify.0.Transform.Rotation=@n3.Result
    ItemToModify.1.Transform.Rotation=@n3.Result
    Weight=@n1.OutputParam
   )
   > n4 = &RigUnit_SphericalPoseReader(
    ActiveRegionScaleFactors=(PositiveWidth=1,NegativeWidth=1,PositiveHeight=1,NegativeHeight=1)
    ActiveRegionSize=0.2
    Debug=(bDrawDebug=false,DebugScale=25,DebugSegments=20,DebugThickness=0.25)
    DriverAxis=(X=1,Y=0,Z=0)
    DriverItem=(Type=Bone,Name="head")
    FalloffRegionScaleFactors=(PositiveWidth=1,NegativeWidth=1,PositiveHeight=1,NegativeHeight=1)
    FalloffSize=0.7
    FlipHeightScaling=false
    FlipWidthScaling=true
    OptionalParentItem=(Type=Bone,Name="clavicle_l")
    RotationOffset=(X=-100,Y=0,Z=7)
   )
   > &RigUnit_ParentConstraint(
    AdvancedSettings=(RotationOrderForFilter=XZY)
    Child=(Type=Bone,Name="dyn_Hair_BackL_03a")
    Filter=(TranslationFilter=(bX=false,bY=false,bZ=false),RotationFilter=(bX=true,bY=true,bZ=true),ScaleFilter=(bX=True,bY=True,bZ=True))
    Parents=((Item=(Type=Bone,Name="dyn_Hair_BackL_03"),Weight=0.3),(Item=(Type=Bone,Name="clavicle_l"),Weight=0.7))
    Weight=@n4.OutputParam
   )
   > &RigUnit_ModifyTransforms(
    ItemToModify=((Item=(Type=Bone,Name="dyn_Hair_BackL_03a"),Transform=(Rotation=(X=0,Y=0,Z=0,W=1),Translation=(X=1,Y=-1,Z=2),Scale3D=(X=1,Y=1,Z=1))))
    ItemToModify.0.Transform.Rotation=&RigVMFunction_MathQuaternionFromEuler(Euler=(X=20,Y=-30,Z=20)).Result
    Weight=@n4.OutputParam
   )
   &RigUnit_ParentConstraint(
    AdvancedSettings=(RotationOrderForFilter=XZY)
    Child=(Type=Bone,Name="dyn_Hair_BackR_01")
    Filter=(TranslationFilter=(bX=true,bY=true,bZ=true),RotationFilter=(bX=true,bY=true,bZ=true),ScaleFilter=(bX=True,bY=True,bZ=True))
    Parents=((Item=(Type=Bone,Name="dyn_Hair_Back_01"),Weight=0),(Item=(Type=Bone,Name="head"),Weight=1))
    Weight=1
    ExecutePin=@SequenceAggregate.C
   )
   > &RigUnit_ParentConstraint(
    AdvancedSettings=(RotationOrderForFilter=XZY)
    Child=(Type=Bone,Name="dyn_Hair_BackR_02")
    Filter=(TranslationFilter=(bX=True,bY=True,bZ=True),RotationFilter=(bX=true,bY=true,bZ=true),ScaleFilter=(bX=True,bY=True,bZ=True))
    Parents=((Item=(Type=Bone,Name="dyn_Hair_Back_02"),Weight=0),(Item=(Type=Bone,Name="head"),Weight=1))
    Weight=1
   )
   > &RigUnit_ParentConstraint(
    AdvancedSettings=(RotationOrderForFilter=XZY)
    Child=(Type=Bone,Name="dyn_Hair_BackR_03")
    Filter=(TranslationFilter=(bX=True,bY=True,bZ=True),RotationFilter=(bX=true,bY=true,bZ=true),ScaleFilter=(bX=True,bY=True,bZ=True))
    Parents=((Item=(Type=Bone,Name="dyn_Hair_Back_03"),Weight=0.3),(Item=(Type=Bone,Name="head"),Weight=0.7))
    Weight=1
   )
   > &RigUnit_ParentConstraint(
    AdvancedSettings=(RotationOrderForFilter=XZY)
    Child=(Type=Bone,Name="dyn_Hair_BackR_04")
    Filter=(TranslationFilter=(bX=True,bY=True,bZ=True),RotationFilter=(bX=true,bY=true,bZ=true),ScaleFilter=(bX=True,bY=True,bZ=True))
    Parents=((Item=(Type=Bone,Name="dyn_Hair_Back_04"),Weight=1),(Item=(Type=Bone,Name="spine_05"),Weight=0))
    Weight=1
   )
   > n5 = &RigUnit_SphericalPoseReader(
    ActiveRegionScaleFactors=(PositiveWidth=0.6,NegativeWidth=0.6,PositiveHeight=0.6,NegativeHeight=0.6)
    ActiveRegionSize=0.1
    Debug=(bDrawDebug=false,DebugScale=25,DebugSegments=20,DebugThickness=0.25)
    DriverAxis=(X=-1,Y=0,Z=0)
    DriverItem=(Type=Bone,Name="upperarm_r")
    FalloffRegionScaleFactors=(PositiveWidth=1,NegativeWidth=0.6,PositiveHeight=0.6,NegativeHeight=0.6)
    FalloffSize=0.6
    FlipHeightScaling=false
    FlipWidthScaling=true
    OptionalParentItem=(Type=Bone,Name="spine_05")
    RotationOffset=(X=80,Y=-30,Z=-10)
   )
   > &RigUnit_ParentConstraint(
    AdvancedSettings=(RotationOrderForFilter=XZY)
    Child=(Type=Bone,Name="dyn_Hair_BackR_03")
    Filter=(TranslationFilter=(bX=false,bY=false,bZ=false),RotationFilter=(bX=true,bY=true,bZ=true),ScaleFilter=(bX=True,bY=True,bZ=True))
    Parents=((Item=(Type=Bone,Name="dyn_Hair_Back_03"),Weight=0.3),(Item=(Type=Bone,Name="upperarm_r"),Weight=0.7))
    Weight=@n5.OutputParam
   )
   n6 = &RigVMFunction_MathQuaternionFromEuler(Euler=(X=-20,Y=10,Z=0), RotationOrder=YXZ)
   > &RigUnit_ModifyTransforms(
    ItemToModify="((Item=(Type=Bone,Name=\"dyn_Hair_BackR_02\"),Transform=(Rotation=(X=0,Y=0,Z=0,W=1),Translation=(X=2,Y=-1,Z=1),Scale3D=(X=1,Y=1,Z=1))),(Item=(Type=Bone,Name=\"dyn_Hair_BackR_04\"),Transform=(Rotation=(X=0,Y=0,Z=0,W=1),Translation=(X=0,Y=0,Z=0),Scale3D=(X=1,Y=1 ...(truncated)"
    ItemToModify.0.Transform.Rotation=@n6.Result
    ItemToModify.1.Transform.Rotation=@n6.Result
    Weight=@n5.OutputParam
   )
   n7 = &RigVMFunction_MathQuaternionFromEuler(Euler=(X=0,Y=0,Z=5))
   > &RigUnit_ModifyTransforms(
    ItemToModify="((Item=(Type=Bone,Name=\"dyn_Hair_Back_03\"),Transform=(Rotation=(X=0,Y=0,Z=0,W=1),Translation=(X=0,Y=0,Z=1),Scale3D=(X=1,Y=1,Z=1))),(Item=(Type=Bone,Name=\"dyn_Hair_Back_04\"),Transform=(Rotation=(X=0,Y=0,Z=0,W=1),Translation=(X=0,Y=0,Z=0),Scale3D=(X=1,Y=1,Z= ...(truncated)"
    ItemToModify.0.Transform.Rotation=@n7.Result
    ItemToModify.1.Transform.Rotation=@n7.Result
    Weight=@n5.OutputParam
   )
   > n8 = &RigUnit_SphericalPoseReader(
    ActiveRegionScaleFactors=(PositiveWidth=1,NegativeWidth=1,PositiveHeight=1,NegativeHeight=1)
    ActiveRegionSize=0.2
    Debug=(bDrawDebug=false,DebugScale=25,DebugSegments=20,DebugThickness=0.25)
    DriverAxis=(X=1,Y=0,Z=0)
    DriverItem=(Type=Bone,Name="head")
    FalloffRegionScaleFactors=(PositiveWidth=1,NegativeWidth=1,PositiveHeight=1.5,NegativeHeight=1.5)
    FalloffSize=0.45
    FlipHeightScaling=false
    FlipWidthScaling=true
    OptionalParentItem=(Type=Bone,Name="clavicle_r")
    RotationOffset=(X=-80,Y=0,Z=7)
   )
   > &RigUnit_ModifyTransforms(
    ItemToModify="((Item=(Type=Bone,Name=\"dyn_Hair_BackR_01\"),Transform=(Rotation=(X=0.0076,Y=-0.0868,Z=0.0868,W=0.9924),Translation=(X=-2,Y=0,Z=2),Scale3D=(X=1,Y=1,Z=1))),(Item=(Type=Bone,Name=\"dyn_Hair_Back_01\"),Transform=(Rotation=(X=0,Y=0,Z=0,W=1),Translation=(X=-2,Y=0, ...(truncated)"
    Weight=@n8.OutputParam
   )
  ]}
 RigVMModel Construction Event Graph {links [], nodes []}
]
hierarchy [
 root {loc=<0 0 0>, rot=<0 0 0 1>}
 pelvis {parent=root, loc=<0 -0.5301 98.0571>, rot=<0.0127 0.707 -0.0127 -0.707>}
 spine_01 {parent=pelvis, loc=<6.1208 -0.9328 0>, rot=<0 0 0.034 -0.9994>}
 spine_02 {parent=spine_01, loc=<6.5996 -0.2095 0>, rot=<0 0 0.0238 -0.9997>}
 spine_03 {parent=spine_02, loc=<6.4914 -0.1237 0>, rot=<0 0 0.0149 -0.9999>}
 spine_04 {parent=spine_03, loc=<6.9598 0.0794 0>, rot=<0 0 -0.0132 -0.9999>}
 spine_05 {parent=spine_04, loc=<6.9987 0 0>, rot=<0 0 -0.0015 -1>}
 neck_01 {parent=spine_05, loc=<11.3138 0.1529 0>, rot=<0 0 -0.2003 -0.9797>}
 neck_02 {parent=neck_01, loc=<4.5627 0 0>, rot=<0 0 0.0504 -0.9987>}
 head {parent=neck_02, loc=<5.3914 0 0>, rot=<0 0 0.115 -0.9934>}
 dyn_Hair_Back_01 {parent=head, loc=<8.5016 -1.2653 0>, rot=<0.3093 -0.6359 0.6359 -0.3093>}
 dyn_Hair_Back_02 {parent=dyn_Hair_Back_01, loc=<4.2813 0 0>, rot=<0 -0.3708 0 -0.9287>}
 dyn_Hair_Back_03 {parent=dyn_Hair_Back_02, loc=<5.5092 0 0>, rot=<0 0.0788 0 -0.9969>}
 dyn_Hair_Back_04 {parent=dyn_Hair_Back_03, loc=<6.5333 0 0>, rot=<0 0.0745 0 -0.9972>}
 dyn_Hair_Back_03a_01 {parent=dyn_Hair_Back_03, loc=<0 0 2.4576>, rot=<0 -0.0619 0 -0.9981>}
 dyn_Hair_Back_03a_02 {parent=dyn_Hair_Back_03a_01, loc=<2.805 0 0>, rot=<0 0.0451 0 -0.999>}
 dyn_Hair_Back_03a_03 {parent=dyn_Hair_Back_03a_02, loc=<3.1303 0 0.0519>, rot=<0 0.1302 0 -0.9915>}
 dyn_Hair_Back_03a_04 {parent=dyn_Hair_Back_03a_03, loc=<2.6825 0 0>, rot=<0 0 0 1>}
 dyn_Hair_BackL_01 {parent=head, loc=<7.9722 -1.0353 -5.5286>, rot=<0.3093 -0.6359 0.6359 -0.3093>}
 dyn_Hair_BackL_02 {parent=dyn_Hair_BackL_01, loc=<4.2813 0 0>, rot=<0 -0.3708 0 -0.9287>}
 dyn_Hair_BackL_03 {parent=dyn_Hair_BackL_02, loc=<5.5092 0 0>, rot=<-0.0029 -0.0411 -0.0295 -0.9987>}
 dyn_Hair_BackL_04 {parent=dyn_Hair_BackL_03, loc=<5.6073 0 0>, rot=<0.1911 0.1673 -0.0363 -0.9665>}
 dyn_Hair_BackL_03a {parent=dyn_Hair_BackL_03, loc=<4.1536 0.7116 -2.2734>, rot=<0.7207 0.1112 0.1607 -0.6651>}
 dyn_Hair_BackR_01 {parent=head, loc=<8.1854 -0.439 5.0959>, rot=<0.3093 -0.6359 0.6359 -0.3093>}
 dyn_Hair_BackR_02 {parent=dyn_Hair_BackR_01, loc=<4.2813 0 0>, rot=<0 -0.3708 0 -0.9287>}
 dyn_Hair_BackR_03 {parent=dyn_Hair_BackR_02, loc=<5.5092 0 0>, rot=<0.0006 0.035 0.0032 -0.9994>}
 dyn_Hair_BackR_04 {parent=dyn_Hair_BackR_03, loc=<7.2027 0 0>, rot=<-0.2756 0.1247 0.0205 -0.9529>}
 dyn_Hair_BackR_03a {parent=dyn_Hair_BackR_03, loc=<3.8752 -1.67 -2.3098>, rot=<-0.6539 0.087 -0.08 -0.7473>}
 clavicle_l {parent=spine_05, loc=<5.4831 -0.1053 -3.2946>, rot=<0.0465 -0.7289 -0.009 -0.6829>}
 upperarm_l {parent=clavicle_l, loc=<9.8658 -1.2136 -0.4105>, rot=<0.0164 -0.3245 -0.0269 -0.9454>}
 lowerarm_l {parent=upperarm_l, loc=<26.465 -0.8345 0.259>, rot=<0.0036 0.0161 -0.2298 -0.9731>}
 hand_l {parent=lowerarm_l, loc=<22.7355 -0.3203 -0.6105>, rot=<-0.6784 -0.0223 0.0735 -0.7307>}
 thumb_01_l {parent=hand_l, loc=<1.3118 -0.3191 -2.459>, rot=<0.6546 -0.4042 0.0664 -0.6354>}
 thumb_02_l {parent=thumb_01_l, loc=<5.9336 -0.0448 0.0401>, rot=<-0.0037 0.0213 0.298 -0.9543>}
 thumb_03_l {parent=thumb_02_l, loc=<2.2223 0.3682 -0.0907>, rot=<0.0118 -0.013 0.0066 -0.9998>}
 index_metacarpal_l {parent=hand_l, loc=<1.6535 0.0183 -2.96>, rot=<0.0158 -0.0124 -0.0173 -0.9996>}
 index_01_l {parent=index_metacarpal_l, loc=<8.3008 0.0357 -0.0524>, rot=<-0.0044 0.0485 0.0924 -0.9945>}
 index_02_l {parent=index_01_l, loc=<3.3587 -0.3004 0.0452>, rot=<-0.002 0.0027 0.3225 -0.9466>}
 index_03_l {parent=index_02_l, loc=<2.5578 -0.123 -0.0252>, rot=<-0.002 0.0067 0.2558 -0.9667>}
 middle_metacarpal_l {parent=hand_l, loc=<2.1856 -0.1725 -0.5984>, rot=<0.006 0.0168 -0.0329 -0.9993>}
 middle_01_l {parent=middle_metacarpal_l, loc=<7.3275 -0.2447 0.0129>, rot=<-0.0208 0.0042 0.0858 -0.9961>}
 middle_02_l {parent=middle_01_l, loc=<4.0123 0.1401 0.1175>, rot=<-0.0218 0.0258 0.3512 -0.9357>}
 middle_03_l {parent=middle_02_l, loc=<2.6077 0.0534 0.1373>, rot=<0.0082 0.0188 0.1708 -0.9851>}
 ring_metacarpal_l {parent=hand_l, loc=<2.2074 -0.3625 1.3638>, rot=<-0.0284 0.0722 -0.0458 -0.9959>}
 ring_01_l {parent=ring_metacarpal_l, loc=<7.0295 -0.6197 -0.1446>, rot=<-0.0489 -0.0252 0.1892 -0.9804>}
 ring_02_l {parent=ring_01_l, loc=<3.6174 0.4647 -0.1027>, rot=<0 0 0.2806 -0.9598>}
 ring_03_l {parent=ring_02_l, loc=<2.4152 0.114 0.0199>, rot=<0 0 0.1189 -0.9929>}
 pinky_metacarpal_l {parent=hand_l, loc=<1.7616 -0.9753 3.3814>, rot=<-0.0397 0.0814 -0.069 -0.9935>}
 pinky_01_l {parent=pinky_metacarpal_l, loc=<7.1709 -0.4831 -0.1094>, rot=<-0.0654 -0.0035 0.2242 -0.9723>}
 pinky_02_l {parent=pinky_01_l, loc=<2.8274 0.0893 -0.1043>, rot=<0.0037 0.0011 0.2854 -0.9584>}
 pinky_03_l {parent=pinky_02_l, loc=<2.2046 0.0307 -0.0078>, rot=<-0.0002 0.0098 0.1445 -0.9895>}
 lowerarm_twist_02_l {parent=lowerarm_l, loc=<10.6707 0 0>, rot=<0 0 0 1>}
 lowerarm_twist_01_l {parent=lowerarm_l, loc=<18.316 0 0>, rot=<0 0 0 1>}
 upperarm_twist_01_l {parent=upperarm_l, loc=<7.2135 -0.1134 0>, rot=<0 0 0 1>}
 upperarm_twist_02_l {parent=upperarm_l, loc=<19.7998 -0.1134 0>, rot=<0 0 0 1>}
 clavicle_r {parent=spine_05, loc=<5.4831 -0.1053 3.2946>, rot=<0.7289 0.0465 -0.6829 0.009>}
 upperarm_r {parent=clavicle_r, loc=<-9.8658 1.2136 0.4105>, rot=<0.0164 -0.3245 -0.0269 -0.9454>}
 lowerarm_r {parent=upperarm_r, loc=<-26.465 0.8345 -0.259>, rot=<0.0036 0.0161 -0.2298 -0.9731>}
 hand_r {parent=lowerarm_r, loc=<-22.7355 0.3203 0.6106>, rot=<-0.6784 -0.0223 0.0735 -0.7307>}
 thumb_01_r {parent=hand_r, loc=<-1.3118 0.3191 2.459>, rot=<0.6546 -0.4042 0.0664 -0.6354>}
 thumb_02_r {parent=thumb_01_r, loc=<-5.9336 0.0448 -0.0401>, rot=<-0.0037 0.0213 0.298 -0.9543>}
 thumb_03_r {parent=thumb_02_r, loc=<-2.2223 -0.3682 0.0907>, rot=<0.0118 -0.013 0.0066 -0.9998>}
 index_metacarpal_r {parent=hand_r, loc=<-1.6535 -0.0183 2.96>, rot=<0.0158 -0.0124 -0.0173 -0.9996>}
 index_01_r {parent=index_metacarpal_r, loc=<-8.3008 -0.0357 0.0524>, rot=<-0.0044 0.0485 0.0924 -0.9945>}
 index_02_r {parent=index_01_r, loc=<-3.3587 0.3004 -0.0452>, rot=<-0.002 0.0027 0.3225 -0.9466>}
 index_03_r {parent=index_02_r, loc=<-2.5578 0.123 0.0252>, rot=<-0.002 0.0067 0.2558 -0.9667>}
 middle_metacarpal_r {parent=hand_r, loc=<-2.1856 0.1725 0.5984>, rot=<0.006 0.0168 -0.0329 -0.9993>}
 middle_01_r {parent=middle_metacarpal_r, loc=<-7.3275 0.2447 -0.0129>, rot=<-0.0208 0.0042 0.0858 -0.9961>}
 middle_02_r {parent=middle_01_r, loc=<-4.0123 -0.1401 -0.1175>, rot=<-0.0218 0.0258 0.3512 -0.9357>}
 middle_03_r {parent=middle_02_r, loc=<-2.6077 -0.0534 -0.1373>, rot=<0.0082 0.0188 0.1708 -0.9851>}
 ring_metacarpal_r {parent=hand_r, loc=<-2.2074 0.3625 -1.3638>, rot=<-0.0284 0.0722 -0.0458 -0.9959>}
 ring_01_r {parent=ring_metacarpal_r, loc=<-7.0295 0.6197 0.1446>, rot=<-0.0489 -0.0252 0.1892 -0.9804>}
 ring_02_r {parent=ring_01_r, loc=<-3.6174 -0.4647 0.1027>, rot=<0 0 0.2806 -0.9598>}
 ring_03_r {parent=ring_02_r, loc=<-2.4152 -0.114 -0.0199>, rot=<0 0 0.1189 -0.9929>}
 pinky_metacarpal_r {parent=hand_r, loc=<-1.7616 0.9753 -3.3814>, rot=<-0.0397 0.0814 -0.069 -0.9935>}
 pinky_01_r {parent=pinky_metacarpal_r, loc=<-7.1709 0.4831 0.1094>, rot=<-0.0654 -0.0035 0.2242 -0.9723>}
 pinky_02_r {parent=pinky_01_r, loc=<-2.8274 -0.0893 0.1043>, rot=<0.0037 0.0011 0.2854 -0.9584>}
 pinky_03_r {parent=pinky_02_r, loc=<-2.2046 -0.0307 0.0078>, rot=<-0.0002 0.0098 0.1445 -0.9895>}
 lowerarm_twist_02_r {parent=lowerarm_r, loc=<-10.6707 0 0>, rot=<0 0 0 -1>}
 lowerarm_twist_01_r {parent=lowerarm_r, loc=<-18.316 0 0>, rot=<0 0 0 -1>}
 upperarm_twist_01_r {parent=upperarm_r, loc=<-7.2135 0.1134 0>, rot=<0 0 0 -1>}
 upperarm_twist_02_r {parent=upperarm_r, loc=<-19.7998 0.1134 0>, rot=<0 0 0 -1>}
 dyn_Hair_Back_01_Null {parent=pelvis, type=Null, loc=<62.3385 -4.7288 -0.3258>, rot=<0 0 0.1148 -0.9934>}
 dyn_Hair_Back_01_ctrl {controlType=EULER_TRANSFORM, parent=dyn_Hair_Back_01_Null, type=Control, loc=<0 0 0>, rot=<0 0 0 1>}
 dyn_Hair_Back_02_Null {parent=pelvis, type=Null, loc=<59.5492 -7.9768 -0.3258>, rot=<0 0 -0.6508 0.7593>}
 dyn_Hair_Back_02_ctrl {controlType=EULER_TRANSFORM, parent=dyn_Hair_Back_02_Null, type=Control, loc=<0 0 0>, rot=<0 0 0 1>}
 dyn_Hair_Back_03_Null {parent=pelvis, type=Null, loc=<52.6768 -8.7618 -0.3258>, rot=<0 0 0.6478 -0.7618>}
 dyn_Hair_Back_03_ctrl {controlType=EULER_TRANSFORM, parent=dyn_Hair_Back_03_Null, type=Control, loc=<0 0 0>, rot=<0 0 0 1>}
 dyn_Hair_Back_04_Null {parent=pelvis, type=Null, loc=<44.2983 -12.3291 -0.6812>, rot=<0 0 -0.6478 0.7618>}
 dyn_Hair_Back_04_ctrl {controlType=EULER_TRANSFORM, parent=dyn_Hair_Back_04_Null, type=Control, loc=<0 0 0>, rot=<0 0 0 1>}
 thigh_l {parent=pelvis, loc=<-1.0855 -0.2109 -8.7053>, rot=<0.0009 0.0355 0.0155 -0.9993>}
 calf_l {parent=thigh_l, loc=<-42.5284 -0.0019 -0.7297>, rot=<0.0006 -0.0161 -0.0391 -0.9991>}
 foot_l {parent=calf_l, loc=<-46.7401 -2.5014 -0.592>, rot=<0.0427 -0.0238 0.0353 -0.9982>}
 ball_l {parent=foot_l, loc=<-3.5658 10.8066 -0.1544>, rot=<0.0001 -0.0001 -0.7063 -0.7079>}
 calf_twist_02_l {parent=calf_l, loc=<-19.1546 -1.7943 -0.2638>, rot=<0.0237 0.0095 -0.0035 -0.9997>}
 calf_twist_01_l {parent=calf_l, loc=<-35.8483 -1.9197 -0.5794>, rot=<0.0237 0.0095 -0.0035 -0.9997>}
 thigh_twist_01_l {parent=thigh_l, loc=<-27.952 0 0>, rot=<0.0013 0 0 -1>}
 thigh_r {parent=pelvis, loc=<-1.0855 -0.2109 8.7053>, rot=<0.0355 -0.0009 0.9993 0.0155>}
 calf_r {parent=thigh_r, loc=<42.5284 0.0019 0.7297>, rot=<0.0006 -0.0161 -0.0391 -0.9991>}
 foot_r {parent=calf_r, loc=<46.7401 2.5014 0.592>, rot=<0.0427 -0.0238 0.0353 -0.9982>}
 ball_r {parent=foot_r, loc=<3.5658 -10.8066 0.1544>, rot=<0.0001 -0.0001 -0.7063 -0.7079>}
 calf_twist_02_r {parent=calf_r, loc=<19.1546 1.7943 0.2638>, rot=<0.0237 0.0095 -0.0035 -0.9997>}
 calf_twist_01_r {parent=calf_r, loc=<35.8483 1.9197 0.5794>, rot=<0.0237 0.0095 -0.0035 -0.9997>}
 thigh_twist_01_r {parent=thigh_r, loc=<27.952 0 0>, rot=<0.0013 0 0 -1>}
]
variables [
 dyn_Hair_BackL_ctrls {
  type="TArray<FRigElementKey>"
  default=((Type=Control,Name="dyn_Hair_BackL_02"),(Type=Control,Name="dyn_Hair_BackL_03"),(Type=Control,Name="dyn_Hair_BackL_04"),(Type=Control,Name="dyn_Hair_BackL_05"),(Type=Control,Name="dyn_Hair_BackL_06"))
 }
 dyn_Hair_BackL_pose {type=FRigPose, default=(Elements=,HierarchyTopologyVersion=-1,PoseHash=-1)}
 dyn_Hair_BackR_ctrls {
  type="TArray<FRigElementKey>"
  default=((Type=Control,Name="dfrm_arm_spline_01_r_ctrl"),(Type=Control,Name="dfrm_arm_spline_03_r_ctrl"),(Type=Control,Name="dfrm_arm_spline_05_r_ctrl"),(Type=Control,Name="dfrm_arm_spline_07_r_ctrl"),(Type=Control,Name="dfrm_arm_spline_09_r_ctrl"))
 }
 dyn_Hair_BackR_pose {type=FRigPose, default=(Elements=,HierarchyTopologyVersion=-1,PoseHash=-1)}
 dyn_Hair_Back_ctrls {
  type="TArray<FRigElementKey>"
  default=((Type=Control,Name="dyn_Hair_BackL_02"),(Type=Control,Name="dyn_Hair_BackL_03"),(Type=Control,Name="dyn_Hair_BackL_04"),(Type=Control,Name="dyn_Hair_BackL_05"),(Type=Control,Name="dyn_Hair_BackL_06"))
 }
 dyn_Hair_Back_pose {type=FRigPose, default=(Elements=,HierarchyTopologyVersion=-1,PoseHash=-1)}
]
