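# Physics asset for the ToadCycle female character's face accessory mesh, covering its dynamic back-hair bones.
# Left, center and right capsule chains hang from the head with mostly narrow swing and twist limits, and each side adds a 03a branch.
# Cross-links tie side chains to the center at lower segments, locked on the right but free on the left.
# Head, neck, upper spine, clavicle and scarf capsules provide body volumes, joined by free constraints.
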PhysicsAsset /BRCosmetics/Characters/Player/Female/Medium/Bodies/F_MED_ToadCycle/Meshes/Parts/F_MED_ToadCycle_FaceAcc_Physics
bodies {
 clavicle_l [{
   shapeType=Capsule
   length=0
   radius=7.5
   shapeName=clavicle_l_capsule
   center=<5.4757 2.4563 2.6426>
   rotation=<rot -0.0039 -89.9997 84.6123>
  }]
 clavicle_r [{shapeType=Capsule, length=1.25, radius=9.75, shapeName=scarf_capsule, center=<-11.1402 -0.4848 4.0616>, rotation=<rot 90 0.0001 0>}]
 dyn_Hair_BackL_01 [{shapeType=Capsule, length=2.5, radius=2.5, shapeName=dyn_Hair_BackL_01_capsule, rotation=<rot 0 -90 -90.305>}]
 dyn_Hair_BackL_02 [{
   shapeType=Capsule
   length=2
   radius=3
   shapeName=dyn_Hair_BackL_02_capsule
   center=<1.8073 -0.0237 4.4401>
   rotation=<rot -0.0003 0 -90.3053>
  }]
 dyn_Hair_BackL_03 [{
   shapeType=Capsule
   length=2
   radius=3
   shapeName=dyn_Hair_BackL_03_capsule
   center=<1.6185 -0.0115 2.1518>
   rotation=<rot -0.0003 0 -90.3053>
  }]
 dyn_Hair_BackL_03a [{
   shapeType=Capsule
   length=5
   radius=3
   shapeName=dyn_Hair_BackL_03a_capsule
   center=<0.8026 -1.2355 1.4459>
   rotation=<rot 0.2323 49.5346 -90.1981>
  }]
 dyn_Hair_BackL_04 [{
   shapeType=Capsule
   length=2.25
   radius=4.75
   shapeName=dyn_Hair_Back_04_capsule
   center=<2.7929 -0.029 5.4381>
   rotation=<rot -0.0003 0 -90.3053>
  }]
 dyn_Hair_BackR_01 [{shapeType=Capsule, length=2.5, radius=2.5, shapeName=dyn_Hair_BackR_01_capsule, rotation=<rot 0 -90 -90.305>}]
 dyn_Hair_BackR_02 [{
   shapeType=Capsule
   length=2
   radius=3
   shapeName=dyn_Hair_BackR_02_capsule
   center=<1.8073 -0.0237 4.4401>
   rotation=<rot -0.0003 0 -90.3053>
  }]
 dyn_Hair_BackR_03 [{
   shapeType=Capsule
   length=2
   radius=3
   shapeName=dyn_Hair_BackR_03_capsule
   center=<1.6185 -0.0115 2.1518>
   rotation=<rot -0.0003 0 -90.3053>
  }]
 dyn_Hair_BackR_03a [{
   shapeType=Capsule
   length=3
   radius=5
   shapeName=dyn_Hair_BackL_03a_capsule
   center=<0.0951 2.2303 1.8717>
   rotation=<rot 0.1436 151.9408 -89.7306>
  }]
 dyn_Hair_BackR_04 [{
   shapeType=Capsule
   length=2.25
   radius=4.75
   shapeName=dyn_Hair_Back_04_capsule
   center=<2.7929 -0.029 5.4381>
   rotation=<rot -0.0003 0 -90.3053>
  }]
 dyn_Hair_Back_01 [{shapeType=Capsule, length=2.5, radius=2.5, shapeName=dyn_Hair_Back_01_capsule, rotation=<rot 0 -90 -90.305>}]
 dyn_Hair_Back_02 [{
   shapeType=Capsule
   length=3.75
   radius=3
   shapeName=dyn_Hair_Back_02_capsule
   center=<1.8073 -0.0237 4.4401>
   rotation=<rot 0.3053 90 -90.0004>
  }]
 dyn_Hair_Back_03 [{shapeType=Capsule, length=4, radius=3, shapeName=dyn_Hair_Back_03_capsule, center=<2.7929 -0.0164 3.078>, rotation=<rot 0.3053 90 -104.063>}]
 dyn_Hair_Back_04 [{shapeType=Capsule, length=4, radius=4.75, shapeName=dyn_Hair_Back_04_capsule, center=<2.7929 -0.023 4.31>, rotation=<rot -0.0003 0 -90.3053>}]
 head [{shapeType=Capsule, length=4.373, radius=8.4467, center=<2.4345 1.7522 0>, rotation=<rot -0.1803 -0.005 0>}]
 neck_01 [{shapeType=Capsule, length=4.5, radius=2.755, shapeName=spine_05_capsule, center=<5.5978 6.2109 0.4446>, rotation=<rot 0 -107.7168 0>}]
 neck_01_scarf [{shapeType=Capsule, length=4.5, radius=7.005, shapeName=spine_05_capsule, center=<1.134 3.1478 0>, rotation=<rot 0 -107.7168 0>}]
 scarf [{shapeType=Capsule, length=8.25, radius=15, shapeName=spine_05_capsule, center=<-2.323 4.4216 0.0401>, rotation=<rot -11.25 -0.0002 0>}]
 spine_05 [{shapeType=Capsule, length=8.25, radius=15, shapeName=spine_05_capsule, center=<-1.8468 3.0932 0>}]
}
constraints [
 {
  bone1Name=head
  bone2Name=neck_01
  swing1LimitDegrees=45
  swing1Motion=Free
  swing2LimitDegrees=45
  swing2Motion=Free
  twistLimitDegrees=45
  twistMotion=Free
 }
 {bone1Name=dyn_Hair_BackL_03a, bone2Name=dyn_Hair_BackL_03, swing1LimitDegrees=10, swing2LimitDegrees=15, twistLimitDegrees=1}
 {bone1Name=dyn_Hair_BackL_02, bone2Name=dyn_Hair_BackL_01, swing1LimitDegrees=10, swing2LimitDegrees=45, twistLimitDegrees=1}
 {
  bone1Name=dyn_Hair_BackL_03
  bone2Name=dyn_Hair_BackL_02
  swing1LimitDegrees=10
  swing1Motion=Locked
  swing2LimitDegrees=45
  twistLimitDegrees=1
  twistMotion=Locked
 }
 {bone1Name=dyn_Hair_BackL_04, bone2Name=dyn_Hair_BackL_03, swing1LimitDegrees=10, swing2LimitDegrees=45, twistLimitDegrees=1}
 {bone1Name=dyn_Hair_BackR_02, bone2Name=dyn_Hair_BackR_01, swing1LimitDegrees=10, swing2LimitDegrees=45, twistLimitDegrees=1}
 {
  bone1Name=dyn_Hair_BackR_03
  bone2Name=dyn_Hair_BackR_02
  swing1LimitDegrees=10
  swing1Motion=Locked
  swing2LimitDegrees=45
  twistLimitDegrees=1
  twistMotion=Locked
 }
 {bone1Name=dyn_Hair_BackR_04, bone2Name=dyn_Hair_BackR_03, swing1LimitDegrees=10, swing2LimitDegrees=45, twistLimitDegrees=35}
 {bone1Name=dyn_Hair_BackR_03a, bone2Name=dyn_Hair_BackR_03, swing1LimitDegrees=1, swing2LimitDegrees=10, twistLimitDegrees=45}
 {bone1Name=dyn_Hair_BackR_01, bone2Name=head, swing1LimitDegrees=10, swing2LimitDegrees=45, twistLimitDegrees=1}
 {bone1Name=dyn_Hair_BackL_01, bone2Name=head, swing1LimitDegrees=10, swing2LimitDegrees=45, twistLimitDegrees=1}
 {bone1Name=dyn_Hair_Back_02, bone2Name=dyn_Hair_Back_01, swing1LimitDegrees=10, swing2LimitDegrees=45, twistLimitDegrees=1}
 {bone1Name=dyn_Hair_Back_01, bone2Name=head, swing1LimitDegrees=10, swing2LimitDegrees=45, twistLimitDegrees=1}
 {bone1Name=dyn_Hair_Back_03, bone2Name=dyn_Hair_Back_02, swing1LimitDegrees=10, swing2LimitDegrees=45, twistLimitDegrees=1}
 {bone1Name=dyn_Hair_Back_04, bone2Name=dyn_Hair_Back_03, swing1LimitDegrees=10, swing2LimitDegrees=45, twistLimitDegrees=1}
 {
  bone1Name=clavicle_l
  bone2Name=spine_05
  swing1LimitDegrees=45
  swing1Motion=Free
  swing2LimitDegrees=45
  swing2Motion=Free
  twistLimitDegrees=45
  twistMotion=Free
 }
 {
  bone1Name=clavicle_r
  bone2Name=spine_05
  swing1LimitDegrees=45
  swing1Motion=Free
  swing2LimitDegrees=45
  swing2Motion=Free
  twistLimitDegrees=45
  twistMotion=Free
 }
 {
  bone1Name=dyn_Hair_BackR_03
  bone2Name=dyn_Hair_Back_03
  swing1LimitDegrees=45
  swing1Motion=Locked
  swing2LimitDegrees=45
  swing2Motion=Locked
  twistLimitDegrees=45
  twistMotion=Locked
 }
 {
  bone1Name=dyn_Hair_BackL_03
  bone2Name=dyn_Hair_Back_03
  swing1LimitDegrees=45
  swing1Motion=Free
  swing2LimitDegrees=45
  swing2Motion=Free
  twistLimitDegrees=45
  twistMotion=Free
 }
 {
  bone1Name=dyn_Hair_BackR_04
  bone2Name=dyn_Hair_Back_04
  swing1LimitDegrees=45
  swing1Motion=Locked
  swing2LimitDegrees=45
  swing2Motion=Locked
  twistLimitDegrees=45
  twistMotion=Locked
 }
 {
  bone1Name=dyn_Hair_BackL_04
  bone2Name=dyn_Hair_Back_04
  swing1LimitDegrees=45
  swing1Motion=Free
  swing2LimitDegrees=45
  swing2Motion=Free
  twistLimitDegrees=45
  twistMotion=Free
 }
 {
  bone1Name=neck_01
  bone2Name=spine_05
  swing1LimitDegrees=45
  swing1Motion=Free
  swing2LimitDegrees=45
  swing2Motion=Free
  twistLimitDegrees=45
  twistMotion=Free
 }
 {
  bone1Name=dyn_Hair_BackR_03a
  bone2Name=dyn_Hair_BackR_04
  swing1LimitDegrees=45
  swing1Motion=Free
  swing2LimitDegrees=45
  swing2Motion=Free
  twistLimitDegrees=45
  twistMotion=Free
 }
 {
  bone1Name=scarf
  bone2Name=spine_05
  swing1LimitDegrees=45
  swing1Motion=Free
  swing2LimitDegrees=45
  swing2Motion=Free
  twistLimitDegrees=45
  twistMotion=Free
 }
 {
  bone1Name=neck_01_scarf
  bone2Name=neck_01
  swing1LimitDegrees=45
  swing1Motion=Free
  swing2LimitDegrees=45
  swing2Motion=Free
  twistLimitDegrees=45
  twistMotion=Free
 }
]
