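# Post-pose AnimBP for the F_MED_ToadCycle female face accessory, adding secondary motion to its scarf and back hair.
# The attached parent's pose runs through a Control Rig, then feeds two scarf branches with bone offsets and separate RigidBody physics assets, selected by the ToadCycleArm meta tag.
# An Anim Dynamics spring chain then moves four hair bones.
# The EventGraph updates tag flags, frontend state and RigidBody gravity via a nativized parameter-asset function.
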AnimBlueprint /BRCosmetics/Characters/Player/Female/Medium/Bodies/F_MED_ToadCycle/Meshes/Parts/F_MED_ToadCycle_FaceAcc_AnimBP
target_skeleton=&/F_MED_ToadCycle_FaceAcc_Skeleton
graphs [
 AnimGraph {nodes [
   n1 = "Transform (Modify) Bone - Bone: scarf"!(
    ComponentPose=Local To Component(LocalPose=Use cached pose 'HairCachePose'())
    bAlphaBoolEnabled:bool=True
    Translation="-1,0,0"
    Rotation="0,0,0"
    Scale="1,1,1"
   ) -> (Pose) {
    alphaInputType=Bool
    rotationMode=BMM_Additive
    rotationSpace=BCS_BoneSpace
    scaleMode=BMM_Additive
    scaleSpace=BCS_BoneSpace
    translationMode=BMM_Additive
    translationSpace=BCS_BoneSpace
    alphaBoolBlend {blendInTime=0.2, blendOutTime=0.2}
    boneToModify {boneName=scarf}
    translation {x=-1}
   }
   n2 = "Transform (Modify) Bone - Bone: scarf"!(ComponentPose=@n1, Alpha:float=1, Translation="1.2,1,0", Rotation="-20,0,0", Scale="1,1,1") -> (Pose) {
    rotationMode=BMM_Additive
    rotationSpace=BCS_BoneSpace
    scaleMode=BMM_Additive
    scaleSpace=BCS_BoneSpace
    translationMode=BMM_Additive
    translationSpace=BCS_BoneSpace
    alphaBoolBlend {blendInTime=0.2, blendOutTime=0.2}
    boneToModify {boneName=scarf}
    rotation {pitch=-20}
    translation=<1.2 1>
   }
   n3 = "Transform (Modify) Bone - Bone: neck_01_scarf"(ComponentPose=@n2, Alpha=1, Translation="0,-2,0", Rotation="0,0,0", Scale="1,1,1") {
    rotationMode=BMM_Additive
    rotationSpace=BCS_BoneSpace
    scaleMode=BMM_Additive
    scaleSpace=BCS_BoneSpace
    translationMode=BMM_Additive
    translationSpace=BCS_BoneSpace
    alphaBoolBlend {blendInTime=0.2, blendOutTime=0.2}
    boneToModify {boneName=neck_01_scarf}
    translation {y=-2}
   }
   n4 = RigidBody!(
    ComponentPose=@n3
    Alpha
    OverrideWorldGravity:Vector=Get Gravity Override Cape() -> (Gravity Override Cape:Vector)
    ExternalForce
    SimSpaceSettings:Sim Space Settings Structure=(WorldAlpha=1,VelocityScaleZ=1,DampingAlpha=0.5,MaxLinearVelocity=10,MaxAngularVelocity=0.5,MaxLinearAcceleration=10,MaxAngularAcceleration=80,ExternalLinearDragV=(X=1,Y=1,Z=1),ExternalLinearVelocity=(X=0,Y=0,Z=0),ExternalAngularVelocity=(X=0,Y=0,Z=0))
   ) -> (Pose) {
    bClampLinearTranslationLimitToRefPose=true
    bOverrideWorldGravity=true
    overridePhysicsAsset=&/F_MED_ToadCycle_FaceAcc_onArm_Physics1
    simulationSpace=BaseBoneSpace
    alphaBoolBlend {blendOption=Linear}
    baseBoneRef {boneName=neck_02}
    componentAppliedLinearAccClamp=<0 0 0>
    overrideWorldGravity {z=-980}
    simSpaceSettings {
     dampingAlpha=0.5
     maxAngularAcceleration=80
     maxAngularVelocity=0.5
     maxLinearAcceleration=10
     maxLinearVelocity=10
     worldAlpha=1
     externalLinearDragV=<1 1 1>
    }
   }
   n5 = Component To Local(ComponentPose=@n4)
   n6 = "Transform (Modify) Bone - Bone: scarf"(
    ComponentPose=Local To Component(LocalPose=Use cached pose 'HairCachePose'())
    Alpha=1
    Translation="4.3,3,0"
    Rotation="0,0,0"
    Scale="1,1,1"
   ) {
    rotationMode=BMM_Additive
    rotationSpace=BCS_BoneSpace
    scaleMode=BMM_Additive
    scaleSpace=BCS_BoneSpace
    translationMode=BMM_Additive
    translationSpace=BCS_BoneSpace
    alphaBoolBlend {blendInTime=0.2, blendOutTime=0.2}
    boneToModify {boneName=scarf}
    translation=<4.3 3>
   }
   n7 = "Transform (Modify) Bone - Bone: neck_01_scarf"(ComponentPose=@n6, Alpha=1, Translation="-2,-2,0", Rotation="0,0,0", Scale="1,1,1") {
    rotationMode=BMM_Additive
    rotationSpace=BCS_BoneSpace
    scaleMode=BMM_Additive
    scaleSpace=BCS_BoneSpace
    translationMode=BMM_Additive
    translationSpace=BCS_BoneSpace
    alphaBoolBlend {blendInTime=0.2, blendOutTime=0.2}
    boneToModify {boneName=neck_01_scarf}
    translation=<-2 -2>
   }
   n8 = RigidBody(
    ComponentPose=@n7
    OverrideWorldGravity=Get Gravity Override Cape()
    SimSpaceSettings=(WorldAlpha=1,VelocityScaleZ=1,DampingAlpha=0.5,MaxLinearVelocity=10,MaxAngularVelocity=0.5,MaxLinearAcceleration=10,MaxAngularAcceleration=80,ExternalLinearDragV=(X=1,Y=1,Z=1),ExternalLinearVelocity=(X=0,Y=0,Z=0),ExternalAngularVelocity=(X=0,Y=0,Z=0))
   ) {
    bClampLinearTranslationLimitToRefPose=true
    bOverrideWorldGravity=true
    overridePhysicsAsset=&/F_MED_ToadCycle_FaceAcc_Physics
    simulationSpace=BaseBoneSpace
    alphaBoolBlend {blendOption=Linear}
    baseBoneRef {boneName=neck_02}
    componentAppliedLinearAccClamp=<0 0 0>
    overrideWorldGravity {z=-980}
    simSpaceSettings {
     dampingAlpha=0.5
     maxAngularAcceleration=80
     maxAngularVelocity=0.5
     maxLinearAcceleration=10
     maxLinearVelocity=10
     worldAlpha=1
     externalLinearDragV=<1 1 1>
    }
   }
   n9 = Component To Local(ComponentPose=@n8)
   n10 = Get On/Off_Arm() -> (On/Off_Arm:bool)
   n11 = Anim Dynamics - Bone: dyn_Hair_Back_03a_01!(
    ComponentPose=Local To Component(LocalPose=Blend Poses by bool(BlendPose_0=@n5, BlendPose_1=@n9, bActiveValue=@n10) {blendType=Linear})
    Alpha=1
    GravityOverride:Vector="0,0,-980"
    ExternalForce="0,0,0"
   ) -> (Pose) {
    angularDampingOverride=1
    angularSpringConstant=800
    bAngularSpring=true
    bOverrideAngularDamping=true
    bUseGravityOverride=true
    gravityScale=0.3
    simulationSpace=RootRelative
    alphaBoolBlend {blendOption=Linear}
    boundBone {boneName=dyn_Hair_Back_03a_01}
    componentAppliedLinearAccClamp=<2000 2000 1600>
    componentLinearAccScale=<0.4 0.4 0.4>
    componentLinearVelScale=<0.45 0.45 0.55>
    gravityOverride {z=-980}
    physicsBodyDefinitions [{
      sphereCollisionRadius=2
      boundBone {boneName=dyn_Hair_Back_03a_01}
      boxExtents=<8 10 12>
      constraintSetup {angularLimitsMax=<2 -2 5>, angularLimitsMin=<-2 -20 -5>, angularTarget=<0 0 0>}
      localJointOffset=<2 0.5 0>
     }]
    relativeSpaceBone {boneName=root}
   }
   n12 = Anim Dynamics - Bone: dyn_Hair_Back_03a_02!(ComponentPose=@n11, Alpha=1, GravityOverride:Vector="0,0,-980", ExternalForce="0,0,0") -> (Pose) {
    angularDampingOverride=1
    angularSpringConstant=800
    bAngularSpring=true
    bOverrideAngularDamping=true
    bUseGravityOverride=true
    gravityScale=0.3
    simulationSpace=RootRelative
    alphaBoolBlend {blendOption=Linear}
    boundBone {boneName=dyn_Hair_Back_03a_02}
    componentAppliedLinearAccClamp=<2000 2000 1600>
    componentLinearAccScale=<0.4 0.4 0.4>
    componentLinearVelScale=<0.45 0.45 0.55>
    gravityOverride {z=-980}
    physicsBodyDefinitions [{
      sphereCollisionRadius=2
      boundBone {boneName=dyn_Hair_Back_03a_02}
      boxExtents=<8 10 12>
      constraintSetup {angularLimitsMax=<2 -2 5>, angularLimitsMin=<-2 -20 -5>, angularTarget=<0 0 0>}
      localJointOffset=<2 0.5 0>
     }]
    relativeSpaceBone {boneName=root}
   }
   n13 = Anim Dynamics - Bone: dyn_Hair_Back_03a_03!(ComponentPose=@n12, Alpha=1, GravityOverride:Vector="0,0,-980", ExternalForce="0,0,0") -> (Pose) {
    angularDampingOverride=1
    angularSpringConstant=800
    bAngularSpring=true
    bOverrideAngularDamping=true
    bUseGravityOverride=true
    gravityScale=0.3
    simulationSpace=RootRelative
    alphaBoolBlend {blendOption=Linear}
    boundBone {boneName=dyn_Hair_Back_03a_03}
    componentAppliedLinearAccClamp=<2000 2000 1600>
    componentLinearAccScale=<0.4 0.4 0.4>
    componentLinearVelScale=<0.45 0.45 0.55>
    gravityOverride {z=-980}
    physicsBodyDefinitions [{
      sphereCollisionRadius=2
      boundBone {boneName=dyn_Hair_Back_03a_03}
      boxExtents=<8 10 12>
      constraintSetup {angularLimitsMax=<2 -2 5>, angularLimitsMin=<-2 -20 -5>, angularTarget=<0 0 0>}
      localJointOffset=<2 0.5 0>
     }]
    relativeSpaceBone {boneName=root}
   }
   n14 = Anim Dynamics - Bone: dyn_Hair_Back_03a_04!(ComponentPose=@n13, Alpha=1, GravityOverride:Vector="0,0,-980", ExternalForce="0,0,0") -> (Pose) {
    angularDampingOverride=1
    angularSpringConstant=800
    bAngularSpring=true
    bOverrideAngularDamping=true
    bUseGravityOverride=true
    gravityScale=0.3
    simulationSpace=RootRelative
    alphaBoolBlend {blendOption=Linear}
    boundBone {boneName=dyn_Hair_Back_03a_04}
    componentAppliedLinearAccClamp=<2000 2000 1600>
    componentLinearAccScale=<0.4 0.4 0.4>
    componentLinearVelScale=<0.45 0.45 0.55>
    gravityOverride {z=-980}
    physicsBodyDefinitions [{
      sphereCollisionRadius=2
      boundBone {boneName=dyn_Hair_Back_03a_04}
      boxExtents=<8 10 12>
      constraintSetup {angularLimitsMax=<2 -2 5>, angularLimitsMin=<-2 -20 -5>, angularTarget=<0 0 0>}
      localJointOffset=<2 0.5 0>
     }]
    relativeSpaceBone {boneName=root}
   }
   Output Pose(Result=Component To Local(ComponentPose=@n14)) {layerGroup=None}
   n15 = Control Rig(Source=Copy Pose From Mesh() {bUseAttachedParent=true}) {controlRigAssetReference {blueprintRigClass=&/F_MED_ToadCycle_FaceAcc_CtrlRig.F_MED_ToadCycle_FaceAcc_CtrlRig_C}}
   Save cached pose 'HairCachePose'(Pose=@n15) {cacheName=HairCachePose}
  ]}
 EventGraph {nodes [
   Event BlueprintUpdateAnimation()
   > n1 = Sequence!(execute) -> (then_0, then_1, then_2:Exec, then_3:Exec)
   n2 = OR Boolean!(
    A=Get bIsPlayingFullBodySlotInFrontEnd() -> (bIsPlayingFullBodySlotInFrontEnd:bool)
    B=Get bIsPreviewingItemInFrontEnd() -> (bIsPreviewingItemInFrontEnd:bool)
    C:bool=Get bIsPlayingEmote()
    D:bool
   ) -> (ReturnValue)
   n3 = AND Boolean(A=Get bIsInFrontEnd(), B=Get bOverrideRBANSimSpaceInFrontEnd() -> (bOverrideRBANSimSpaceInFrontEnd:bool))
   n4 = Get bShouldDisplayFrontendReactivity() -> (bShouldDisplayFrontendReactivity:bool)
   Set is_Frontend And Not emote(
    execute=@n1.then_3
    is_Frontend And Not emote:bool=AND Boolean!(A=OR Boolean!(A=@n3, B=@n4) -> (ReturnValue), B=NOT Boolean(A=@n2), C:bool=Get On/Off_Scarf() -> (On/Off_Scarf:bool)) -> (ReturnValue)
   ) -> (Output_Get:bool)
   n5 = Branch(execute=@n1.then_1, Condition=CharacterItemDefinitionHasAnyMetaTag(Tags=Make Array("[0]"=(TagName="Cosmetics.MetaTags.ToadCycleArm"))))
   > Set On/Off_Arm(On/Off_Arm:bool=true) -> (Output_Get:bool)
   Set On/Off_Arm(execute=@n5.else, On/Off_Arm=false)
   n6 = Branch(execute=@n1.then_2, Condition=CharacterItemDefinitionHasAnyMetaTag(Tags=Make Array("[0]"=(TagName="Cosmetics.MetaTags.ToadCycleSarf"))))
   > Set On/Off_Scarf(On/Off_Scarf:bool=1) -> (Output_Get:bool)
   Set On/Off_Scarf(execute=@n6.else, On/Off_Scarf=false)
   n7 = "Gravity Override State Function (Parameter Asset) (Nativized)"(
    execute=@n1.then_0
    CharacterPartAnimInstance=Self-Reference()
    GameTimeInSeconds=GetGameTimeInSeconds()
    Parameters=&/F_MED_ToadCycle_FaceAcc_Parameters
   )
   > Set Gravity Override Cape(Gravity Override Cape:Vector=@n7) -> (Output_Get:Vector)
   > Set RBAN Settings Out_0(RBAN Settings Out_0:Fort Rigid Body Anim Node Parameters Structure=@n7.OutRigidBodyAnimNodeParameters) -> (Output_Get:Fort Rigid Body Anim Node Parameters Structure)
   n8 = Event OnCharacterCustomizationLoaded()
   > n9 = Sequence()
   Set Character Mesh(execute=@n9.then_1, Character Mesh:Skeletal Mesh Component Object Reference=@n8.CustomBaseMesh) -> (Output_Get:Skeletal Mesh Component Object Reference)
   > n10 = Cast To FortPlayerPawn(Object=TryGetPawnOwner())
   > Set As Fort Player Pawn(As Fort Player Pawn:Fort Player Pawn Object Reference=@n10) -> (Output_Get:Fort Player Pawn Object Reference)
   n11 = CheckForBackbling()
   Set isWearingPackBool(execute=@n9.then_3, isWearingPackBool:bool=OR Boolean!(A=@n11, B=@n11.bIsBackblingACape, C:bool=@n11.bIsBackblingLong) -> (ReturnValue)) -> (Output_Get:bool)
   > Set is_WearingCape(is_WearingCape:bool=@n11.bIsBackblingLong) -> (Output_Get:bool)
   Set Head_Mesh(
    execute=@n9.then_2
    Head_Mesh:Skeletal Mesh Component Object Reference=Get Array Item with EFort Custom Part Type(Array=@n8.CustomMeshParts, PartType=Head)
   ) -> (Output_Get:Skeletal Mesh Component Object Reference)
   GetSkeletalMesh!(self:Custom Character Part Object Reference=Get Array Item with EFort Custom Part Type(Array:Array of Custom Character Part Object References=@n8.CustomParts, PartType=Head) -> (Output:Custom Character Part Object Reference)) -> (ReturnValue:Skeletal Mesh Object Reference)
   AND Boolean!(A=@n11, B=NOT Boolean(A=@n11.bIsBackblingLong)) -> (ReturnValue)
  ]}
]
variables [
 As Fort Player Pawn {type=FortPlayerPawn}
 Character Mesh {type=SkeletalMeshComponent}
 Gravity Override Cape {type=Vector}
 Head_Mesh {type=SkeletalMeshComponent}
 is_Frontend And Not emote {type=boolean}
 is_WearingCape {type=boolean}
 isWearingPackBool {type=boolean}
 On/Off_Arm {type=boolean}
 On/Off_Scarf {type=boolean}
 RBAN Settings Out_0 {type=FortRigidBodyAnimNodeParameters}
]
